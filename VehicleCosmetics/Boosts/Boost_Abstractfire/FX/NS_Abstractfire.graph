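# Niagara system for the Abstractfire vehicle boost cosmetic, trailing ribbons and sprites backward from the car's skeletal mesh sockets.
# The ribbon emitter spawns per filtered socket at quality-scaled rates, tapering and fading over life through two renderers.
# The sprite emitter partitions three spawn groups across sockets, varying size, speed and lifetime, one by game mode.
# Painted variants set primary and R1 colors, switching on frontend preview; sprite colors animate through curves.
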NiagaraSystem /VehicleCosmetics/Boosts/Boost_Abstractfire/FX/NS_Abstractfire
dependencies {
 usedDataInterfaces [
  &NiagaraDataInterfaceCurve
  &NiagaraDataInterfaceColorCurve
  &NiagaraDataInterfaceSpriteRendererInfo
  &NiagaraDataInterfaceSkeletalMesh
 ]
 usedDynamicInputs []
 usedModules [
  &self:SystemSpawnScript.SystemScriptSource.SystemScriptGraph.NiagaraNodeAssignment_0.SetVariables_#1
  &SystemState
 ]
}
emitters {
 NE_VC_Boost_Ribbon_Parent {
  emitterUpdateScript [
   &EmitterState(Inactive Response=::NewEnumerator1, Life Cycle Mode=::NewEnumerator1, Loop Duration=1, UseLoopDelay=false)
   &VC_FilteredSocketCount()
   &VC_RateQualityHandler(Epic Spawn Rate=40, High Spawn Rate=40, Low Spawn Rate=30, Medium Spawn Rate=36, Use Custom Rate?=true)
   &VC_SpawnRibbon_v1(Number of Ribbons=$Emitter.PortCount, SpawnRate=$Emitter.Spawn Rate)
  ]
  particleSpawnScript [
   &VC_PaintedVariantsOverride(
    Black=&MakeCustomLinearColorFromBool(Bool=$User.bIsFrontendPreview, True Color=<rgba 0.006 0.006 0.006 1>, False Color=<rgba 0.008 0.008 0.008 1>)
    BurntSienna=&MakeCustomLinearColorFromBool(Bool=$User.bIsFrontendPreview, True Color=<rgba 0.045 0.0094 0.0023 1>, False Color=<rgba 0.09 0.0259 0.0045 1>)
    Cobalt=&MakeCustomLinearColorFromBool(Bool=$User.bIsFrontendPreview, True Color=<rgba 0.0216 0.0362 0.24 1>, False Color=<rgba 0.036 0.0603 0.4 1>)
    Crimson=&MakeCustomLinearColorFromBool(Bool=$User.bIsFrontendPreview, True Color=<rgba 0.13 0.0022 0 1>, False Color=<rgba 0.21 0.0035 0 1>)
    DefaultColor=&MakeCustomLinearColorFromBool(Bool=$User.bIsFrontendPreview, True Color=<rgba 0.0057 0.0211 0.19 1>, False Color=<rgba 0.0192 0.0443 0.32 1>)
    ForestGreen=&MakeCustomLinearColorFromBool(Bool=$User.bIsFrontendPreview, True Color=<rgba 0 0.12 0 1>, False Color=<rgba 0 0.16 0 1>)
    Grey=&MakeCustomLinearColorFromBool(Bool=$User.bIsFrontendPreview, True Color=<rgba 0.047 0.047 0.047 1>, False Color=<rgba 0.06 0.06 0.06 1>)
    Lime=&MakeCustomLinearColorFromBool(Bool=$User.bIsFrontendPreview, True Color=<rgba 0.1173 0.23 0.0046 1>, False Color=<rgba 0.153 0.3 0.006 1>)
    Orange=&MakeCustomLinearColorFromBool(Bool=$User.bIsFrontendPreview, True Color=<rgba 0.16 0.056 0 1>, False Color=<rgba 0.24 0.088 0 1>)
    Pink=&MakeCustomLinearColorFromBool(Bool=$User.bIsFrontendPreview, True Color=<rgba 0.123 0.0378 0.14 1>, False Color=<rgba 0.2028 0.0667 0.23 1>)
    Purple=&MakeCustomLinearColorFromBool(Bool=$User.bIsFrontendPreview, True Color=<rgba 0.044 0 0.12 1>, False Color=<rgba 0.0773 0 0.16 1>)
    R1 Black:LinearColor=&MakeCustomLinearColorFromBool(Bool=$User.bIsFrontendPreview, True Color=<rgba 0.03 0.03 0.03 1>, False Color=<rgba 0.06 0.06 0.06 1>)
    R1 BurntSienna:LinearColor=&MakeCustomLinearColorFromBool(Bool=$User.bIsFrontendPreview, True Color=<rgba 0.15 0.0313 0.0075 1>, False Color=<rgba 0.3 0.0863 0.015 1>)
    R1 Cobalt:LinearColor=&MakeCustomLinearColorFromBool(Bool=$User.bIsFrontendPreview, True Color=<rgba 0.048 0.0848 0.6 1>, False Color=<rgba 0.0525 0.099 0.75 1>)
    R1 Crimson:LinearColor=&MakeCustomLinearColorFromBool(Bool=$User.bIsFrontendPreview, True Color=<rgba 0.36 0.006 0 1>, False Color=<rgba 0.45 0.0075 0 1>)
    R1 DefaultColor=&MakeCustomLinearColorFromBool(Bool=$User.bIsFrontendPreview, True Color=<rgba 0.0184 0.0478 0.46 1>, False Color=<rgba 0.0372 0.0761 0.62 1>)
    R1 ForestGreen:LinearColor=&MakeCustomLinearColorFromBool(Bool=$User.bIsFrontendPreview, True Color=<rgba 0 0.4 0 1>, False Color=<rgba 0 0.46 0 1>)
    R1 Grey:LinearColor=&MakeCustomLinearColorFromBool(Bool=$User.bIsFrontendPreview, True Color=<rgba 0.15 0.15 0.15 1>, False Color=<rgba 0.2 0.2 0.2 1>)
    R1 Lime:LinearColor=&MakeCustomLinearColorFromBool(Bool=$User.bIsFrontendPreview, True Color=<rgba 0.41 0.6 0.03 1>, False Color=<rgba 0.4442 0.65 0.0325 1>)
    R1 Orange:LinearColor=&MakeCustomLinearColorFromBool(Bool=$User.bIsFrontendPreview, True Color=<rgba 0.55 0.1925 0 1>, False Color=<rgba 0.7 0.2567 0 1>)
    R1 Pink:LinearColor=&MakeCustomLinearColorFromBool(Bool=$User.bIsFrontendPreview, True Color=<rgba 0.35 0.1 0.4 1>, False Color=<rgba 0.5698 0.169 0.65 1>)
    R1 Purple:LinearColor=&MakeCustomLinearColorFromBool(Bool=$User.bIsFrontendPreview, True Color=<rgba 0.1613 0 0.44 1>, False Color=<rgba 0.1933 0 0.4 1>)
    R1 Saffron:LinearColor=&MakeCustomLinearColorFromBool(Bool=$User.bIsFrontendPreview, True Color=<rgba 0.6 0.6 0 1>, False Color=<rgba 0.65 0.65 0 1>)
    R1 SkyBlue:LinearColor=&MakeCustomLinearColorFromBool(Bool=$User.bIsFrontendPreview, True Color=<rgba 0.012 0.2276 0.6 1>, False Color=<rgba 0.0092 0.3097 0.46 1>)
    R1 TitaniumWhite:LinearColor=&MakeCustomLinearColorFromBool(Bool=$User.bIsFrontendPreview, True Color=<rgba 0.6 0.6 0.6 1>, False Color=<rgba 0.65 0.65 0.65 1>)
    Saffron=&MakeCustomLinearColorFromBool(Bool=$User.bIsFrontendPreview, True Color=<rgba 0.2 0.2 0 1>, False Color=<rgba 0.24 0.24 0 1>)
    SkyBlue=&MakeCustomLinearColorFromBool(Bool=$User.bIsFrontendPreview, True Color=<rgba 0.0046 0.0835 0.23 1>, False Color=<rgba 0.0038 0.1279 0.19 1>)
    TitaniumWhite=&MakeCustomLinearColorFromBool(Bool=$User.bIsFrontendPreview, True Color=<rgba 0.2 0.2 0.2 1>, False Color=<rgba 0.25 0.25 0.25 1>)
    Use R1 Painted Overrides=true
   )
   &VC_InitializeRibbon_v1(
    Color=$Particles.Color
    Custom Offset=true
    Enable Skeletal Position=true
    Lifetime=1
    Local Vector Offset=&AddVectorToPosition(Position=<-20 0 0>, Vector=$User.DTCarOffset)
    Offset Amount=1
    Ribbon Width=50
   )
   &AddVelocity(Velocity=<-1600 0 0>) {module=AddVelocity001}
  ]
  particleUpdateScript [
   &ParticleState()
   &InheritSourceMovement(ApplyForce=false, ApplyPosition=true, ApplyPositionFalloff=false)
   &SolveForcesAndVelocity(Clamp Velocity=false, Limit Acceleration=false)
   &VC_UpdateRibbon_v1(Enable Skeletal Position=true)
   &ScaleRibbonWidth(Ribbon Width Scale=&FloatFromCurve(FloatCurve=DataInterface() {propertyValues {Curve {keys [<at 0 1>,<at 1 0.4>]}}}))
   &ScaleColor(
    Scale Alpha=&FloatFromCurve(FloatCurve=DataInterface() {propertyValues {Curve {keys [<at 0.02 0>,<at 0.15 1>,<at 0.75 1>,<at 0.9 0>]}}})
    ScaleA=true
    ScaleRGB=false
   )
   /SetVariables(
    Particles.r1:LinearColor=&MultiplyLinearColor(
     Linear Color=$Particles.R1 Color
     Scale Factor=&MakeVector4FromVectorAndFloat(
      "Vector (RGB)"=&VectorFromFloat(Value=1)
      "Float (Alpha)"=&FloatFromCurve(FloatCurve=DataInterface() {propertyValues {Curve {keys [<at 0.02 0>,<at 0.15 1>,<at 0.75 1>,<at 0.9 0>]}}})
     )
    )
    Particles.width:NiagaraFloat=&Multiply_Float(
     A=&FloatFromCurve(FloatCurve=DataInterface() {propertyValues {Curve {keys [<at 0 1>,<at 1 0.4>]}}})
     B=40
    )
   ) {isSetParameters=true}
  ]
  rendererClasses [&NiagaraRibbonRendererProperties]
  renderers [
   {rendererClass=&NiagaraRibbonRendererProperties}
   {rendererClass=&NiagaraRibbonRendererProperties, rendererIndex=1}
  ]
 }
 NE_VC_Boost_Sprite_Parent {
  emitterUpdateScript [
   &EmitterState(Inactive Response=::NewEnumerator1, Life Cycle Mode=::NewEnumerator1, Loop Duration=2, UseLoopDelay=false)
   &VC_FilteredSocketCount()
   &SpawnRate(Spawn Probability=0.7, SpawnRate=80, Use Spawn Probability=true)
   &SpawnRate(Spawn Probability=0.7, SpawnGroup=1, SpawnRate=30, Use Spawn Probability=true) {module=SpawnRate001}
   &SpawnRate(Spawn Probability=0.8, SpawnGroup=2, SpawnRate=8, Use Spawn Probability=true) {module=SpawnRate002}
  ]
  particleSpawnScript [
   &VC_PaintedVariantsOverride() {enabled=false}
   &InitializeParticle(
    Color Mode=::NewEnumerator0
    Lifetime=&MergeHelper_FloatSwitch(
     0=&RandomRangeFloat(Minimum=1, Maximum=1.2)
     1=&MergeHelper_FloatSwitch(
      0=&RandomRangeFloat(Minimum=1.8, Maximum=2)
      1=&RandomRangeFloat(Minimum=1.3, Maximum=1.6)
      2=&RandomRangeFloat(Minimum=1.3, Maximum=1.6)
      3=?
      Index=$User.GameModeID
      Num=::NewEnumerator1
     )
     2=&RandomRangeFloat(Minimum=1.6, Maximum=2)
     3=?
     Index=&SpawnGroup()
     Num=::NewEnumerator1
    )
    Sprite Rotation Mode=::NewEnumerator2
    Sprite Size Mode=::NewEnumerator3
    Uniform Sprite Size=&MergeHelper_FloatSwitch(
     0=&RandomRangeFloat(Minimum=70, Maximum=100)
     1=&RandomRangeFloat(Minimum=50, Maximum=70)
     2=&RandomRangeFloat(Minimum=70, Maximum=100)
     3=?
     Index=&SpawnGroup()
     Num=::NewEnumerator1
    )
    UsePositionOffset=false
    Write Material Random=true
   )
   &PartitionParticles(Number Of Partitions=$Emitter.PortCount)
   &SkeletalMeshLocation(
    Bone / Socket Index=$Particles.PartitionParticles.Partition
    Bone / Socket Sampling Mode=::NewEnumerator5
    MaskSpawnGroup=false
    Mesh Sampling Type=::NewEnumerator6
    Skeletal Mesh=$System.SkeletalMesh
   )
   &ShapeLocation(
    MaskSpawnGroup=false
    Non Uniform Scale=&MergeHelper_VectorSwitch(0=<1 1 1>, 1=<1 1 2>, 2=<1 1 1>, 3=?, Index=&SpawnGroup(), Num=::NewEnumerator1)
    Offset=&AddVectorToPosition(Position=<-40 0 0>, Vector=$User.DTCarOffset)
    Offset Coordinate Space=::Local
    Offset Mode=::NewEnumerator0
    Override Seed=false
    Sphere Radius=25
    Sphere Surface Distribution=0.4
    UseOwnerScale=true
   )
   &AddVelocity(Velocity=&MergeHelper_VectorSwitch(0=<-1400 0 0>, 1=<-1200 0 0>, 2=<-1300 0 0>, 3=?, Index=&SpawnGroup(), Num=::NewEnumerator1))
   /SetVariables(Particles.VisibilityTag=$Emitter.SpawnGroup) {isSetParameters=true}
  ]
  particleUpdateScript [
   &VC_PaintedVariantsOverride(
    Black=&MakeCustomLinearColorFromBool(
     Bool=$User.bIsFrontendPreview
     True Color=&ColorFromCurve(ColorCurve=DataInterface() {propertyValues {
        AlphaCurve {keys [<at 0 1 RCIM_Cubic>]}
        BlueCurve {keys [<at 0 0.003>,<at 0.12 0.008>,<at 0.9355 0.018>]}
        GreenCurve {keys [<at 0 0.003>,<at 0.12 0.008>,<at 0.9355 0.018>]}
        RedCurve {keys [<at 0 0.003>,<at 0.12 0.008>,<at 0.9355 0.018>]}
       }})
     False Color=&ColorFromCurve(ColorCurve=DataInterface() {propertyValues {
        AlphaCurve {keys [<at 0 1 RCIM_Cubic>]}
        BlueCurve {keys [<at 0 0.005>,<at 0.12 0.012>,<at 0.85 0.021>]}
        GreenCurve {keys [<at 0 0.005>,<at 0.12 0.012>,<at 0.85 0.021>]}
        RedCurve {keys [<at 0 0.005>,<at 0.12 0.012>,<at 0.85 0.021>]}
       }})
    )
    BurntSienna=&MakeCustomLinearColorFromBool(
     Bool=$User.bIsFrontendPreview
     True Color=&ColorFromCurve(ColorCurve=DataInterface() {propertyValues {
        AlphaCurve {keys [<at 0 1 RCIM_Cubic>]}
        BlueCurve {keys [<at 0 0.0012>,<at 0.12 0.0033>,<at 0.9355 0.0063>]}
        GreenCurve {keys [<at 0 0.0084>,<at 0.12 0.0222>,<at 0.9355 0.0415>]}
        RedCurve {keys [<at 0 0.03>,<at 0.12 0.0802>,<at 0.9355 0.15>]}
       }})
     False Color=&ColorFromCurve(ColorCurve=DataInterface() {propertyValues {
        AlphaCurve {keys [<at 0 1 RCIM_Cubic>]}
        BlueCurve {keys [<at 0 0.002>,<at 0.12 0.0058>,<at 0.85 0.0067>]}
        GreenCurve {keys [<at 0 0.014>,<at 0.12 0.0387>,<at 0.85 0.0442>]}
        RedCurve {keys [<at 0 0.05>,<at 0.12 0.14>,<at 0.85 0.16>]}
       }})
    )
    Cobalt=&MakeCustomLinearColorFromBool(
     Bool=$User.bIsFrontendPreview
     True Color=&ColorFromCurve(ColorCurve=DataInterface() {propertyValues {
        AlphaCurve {keys [<at 0 1 RCIM_Cubic>]}
        BlueCurve {keys [<at 0 0.05>,<at 0.12 0.3>,<at 0.9355 0.5>]}
        GreenCurve {keys [<at 0 0.0065>,<at 0.12 0.0387>,<at 0.9355 0.0646>]}
        RedCurve {keys [<at 0 0.0025>,<at 0.12 0.015>,<at 0.9355 0.025>]}
       }})
     False Color=&ColorFromCurve(ColorCurve=DataInterface() {propertyValues {
        AlphaCurve {keys [<at 0 1 RCIM_Cubic>]}
        BlueCurve {keys [<at 0 0.15>,<at 0.12 0.4>,<at 0.85 0.55>]}
        GreenCurve {keys [<at 0 0.035>,<at 0.12 0.0627>,<at 0.85 0.0862>]}
        RedCurve {keys [<at 0 0.012>,<at 0.12 0.032>,<at 0.85 0.044>]}
       }})
    )
    Crimson=&MakeCustomLinearColorFromBool(
     Bool=$User.bIsFrontendPreview
     True Color=&ColorFromCurve(ColorCurve=DataInterface() {propertyValues {
        AlphaCurve {keys [<at 0 1 RCIM_Cubic>]}
        BlueCurve {keys [<at 0 0>,<at 0.12 0>,<at 0.9355 0>]}
        GreenCurve {keys [<at 0 0.0013>,<at 0.12 0.002>,<at 0.9355 0.0043>]}
        RedCurve {keys [<at 0 0.08>,<at 0.12 0.12>,<at 0.9355 0.26>]}
       }})
     False Color=&ColorFromCurve(ColorCurve=DataInterface() {propertyValues {
        AlphaCurve {keys [<at 0 1 RCIM_Cubic>]}
        BlueCurve {keys [<at 0 0>,<at 0.12 0>,<at 0.85 0>]}
        GreenCurve {keys [<at 0 0.0017>,<at 0.12 0.005>,<at 0.85 0.0058>]}
        RedCurve {keys [<at 0 0.1>,<at 0.12 0.3>,<at 0.85 0.35>]}
       }})
    )
    DefaultColor=&MakeCustomLinearColorFromBool(
     Bool=$User.bIsFrontendPreview
     True Color=&ColorFromCurve(ColorCurve=DataInterface() {propertyValues {
        AlphaCurve {keys [<at 0 1 RCIM_Cubic>]}
        BlueCurve {keys [<at 0 0.05>,<at 0.12 0.25>,<at 0.9355 0.35>]}
        GreenCurve {keys [<at 0 0.0046>,<at 0.12 0.0231>,<at 0.9355 0.0324>]}
        RedCurve {keys [<at 0 0.0005>,<at 0.12 0.0025>,<at 0.9355 0.0035>]}
       }})
     False Color=&ColorFromCurve(ColorCurve=DataInterface() {propertyValues {
        AlphaCurve {keys [<at 0 1 RCIM_Cubic>]}
        BlueCurve {keys [<at 0 0.1>,<at 0.12 0.3185>,<at 0.85 0.38>]}
        GreenCurve {keys [<at 0 0.0092>,<at 0.12 0.0437>,<at 0.85 0.0526>]}
        RedCurve {keys [<at 0 0.001>,<at 0.12 0.0194>,<at 0.85 0.0228>]}
       }})
    )
    ForestGreen=&MakeCustomLinearColorFromBool(
     Bool=$User.bIsFrontendPreview
     True Color=&ColorFromCurve(ColorCurve=DataInterface() {propertyValues {
        AlphaCurve {keys [<at 0 1 RCIM_Cubic>]}
        BlueCurve {keys [<at 0 0.0012>,<at 0.12 0.0032>,<at 0.9355 0.0064>]}
        GreenCurve {keys [<at 0 0.03>,<at 0.12 0.08>,<at 0.9355 0.16>]}
        RedCurve {keys [<at 0 0.0012>,<at 0.12 0.0032>,<at 0.9355 0.0064>]}
       }})
     False Color=&ColorFromCurve(ColorCurve=DataInterface() {propertyValues {
        AlphaCurve {keys [<at 0 1 RCIM_Cubic>]}
        BlueCurve {keys [<at 0 0.002>,<at 0.12 0.006>,<at 0.85 0.0084>]}
        GreenCurve {keys [<at 0 0.05>,<at 0.12 0.15>,<at 0.85 0.21>]}
        RedCurve {keys [<at 0 0.002>,<at 0.12 0.006>,<at 0.85 0.0084>]}
       }})
    )
    Grey=&MakeCustomLinearColorFromBool(
     Bool=$User.bIsFrontendPreview
     True Color=&ColorFromCurve(ColorCurve=DataInterface() {propertyValues {
        AlphaCurve {keys [<at 0 1 RCIM_Cubic>]}
        BlueCurve {keys [<at 0 0.02>,<at 0.12 0.03>,<at 0.9355 0.08>]}
        GreenCurve {keys [<at 0 0.02>,<at 0.12 0.03>,<at 0.9355 0.08>]}
        RedCurve {keys [<at 0 0.02>,<at 0.12 0.03>,<at 0.9355 0.08>]}
       }})
     False Color=&ColorFromCurve(ColorCurve=DataInterface() {propertyValues {
        AlphaCurve {keys [<at 0 1 RCIM_Cubic>]}
        BlueCurve {keys [<at 0 0.04>,<at 0.12 0.1>,<at 0.85 0.13>]}
        GreenCurve {keys [<at 0 0.04>,<at 0.12 0.1>,<at 0.85 0.13>]}
        RedCurve {keys [<at 0 0.04>,<at 0.12 0.1>,<at 0.85 0.13>]}
       }})
    )
    Lime=&MakeCustomLinearColorFromBool(
     Bool=$User.bIsFrontendPreview
     True Color=&ColorFromCurve(ColorCurve=DataInterface() {propertyValues {
        AlphaCurve {keys [<at 0 1 RCIM_Cubic>]}
        BlueCurve {keys [<at 0 0.0024>,<at 0.12 0.0042>,<at 0.9355 0.0078>]}
        GreenCurve {keys [<at 0 0.08>,<at 0.12 0.14>,<at 0.9355 0.26>]}
        RedCurve {keys [<at 0 0.0412>,<at 0.12 0.0721>,<at 0.9355 0.1339>]}
       }})
     False Color=&ColorFromCurve(ColorCurve=DataInterface() {propertyValues {
        AlphaCurve {keys [<at 0 1 RCIM_Cubic>]}
        BlueCurve {keys [<at 0 0.003>,<at 0.12 0.0096>,<at 0.85 0.0108>]}
        GreenCurve {keys [<at 0 0.1>,<at 0.12 0.32>,<at 0.85 0.36>]}
        RedCurve {keys [<at 0 0.0596>,<at 0.12 0.1907>,<at 0.85 0.2145>]}
       }})
    )
    Orange=&MakeCustomLinearColorFromBool(
     Bool=$User.bIsFrontendPreview
     True Color=&ColorFromCurve(ColorCurve=DataInterface() {propertyValues {
        AlphaCurve {keys [<at 0 1 RCIM_Cubic>]}
        BlueCurve {keys [<at 0 0>,<at 0.12 0>,<at 0.9355 0>]}
        GreenCurve {keys [<at 0 0.0208>,<at 0.12 0.1042>,<at 0.9355 0.1458>]}
        RedCurve {keys [<at 0 0.05>,<at 0.12 0.25>,<at 0.9355 0.35>]}
       }})
     False Color=&ColorFromCurve(ColorCurve=DataInterface() {propertyValues {
        AlphaCurve {keys [<at 0 1 RCIM_Cubic>]}
        BlueCurve {keys [<at 0 0>,<at 0.12 0>,<at 0.85 0>]}
        GreenCurve {keys [<at 0 0.025>,<at 0.12 0.1333>,<at 0.85 0.15>]}
        RedCurve {keys [<at 0 0.06>,<at 0.12 0.32>,<at 0.85 0.36>]}
       }})
    )
    Pink=&MakeCustomLinearColorFromBool(
     Bool=$User.bIsFrontendPreview
     True Color=&ColorFromCurve(ColorCurve=DataInterface() {propertyValues {
        AlphaCurve {keys [<at 0 1 RCIM_Cubic>]}
        BlueCurve {keys [<at 0 0.05>,<at 0.12 0.2>,<at 0.9355 0.3>]}
        GreenCurve {keys [<at 0 0.015>,<at 0.12 0.06>,<at 0.9355 0.09>]}
        RedCurve {keys [<at 0 0.0413>,<at 0.12 0.165>,<at 0.9355 0.2475>]}
       }})
     False Color=&ColorFromCurve(ColorCurve=DataInterface() {propertyValues {
        AlphaCurve {keys [<at 0 1 RCIM_Cubic>]}
        BlueCurve {keys [<at 0 0.06>,<at 0.12 0.3>,<at 0.85 0.35>]}
        GreenCurve {keys [<at 0 0.018>,<at 0.12 0.09>,<at 0.85 0.105>]}
        RedCurve {keys [<at 0 0.0495>,<at 0.12 0.2475>,<at 0.85 0.2888>]}
       }})
    )
    Purple=&MakeCustomLinearColorFromBool(
     Bool=$User.bIsFrontendPreview
     True Color=&ColorFromCurve(ColorCurve=DataInterface() {propertyValues {
        AlphaCurve {keys [<at 0 1 RCIM_Cubic>]}
        BlueCurve {keys [<at 0 0.08>,<at 0.12 0.1>,<at 0.9355 0.3>]}
        GreenCurve {keys [<at 0 0>,<at 0.12 0>,<at 0.9355 0>]}
        RedCurve {keys [<at 0 0.0293>,<at 0.12 0.0367>,<at 0.9355 0.11>]}
       }})
     False Color=&ColorFromCurve(ColorCurve=DataInterface() {propertyValues {
        AlphaCurve {keys [<at 0 1 RCIM_Cubic>]}
        BlueCurve {keys [<at 0 0.1>,<at 0.12 0.3>,<at 0.85 0.35>]}
        GreenCurve {keys [<at 0 0>,<at 0.12 0>,<at 0.85 0>]}
        RedCurve {keys [<at 0 0.0483>,<at 0.12 0.145>,<at 0.85 0.1692>]}
       }})
    )
    R1 Black:LinearColor=&MakeCustomLinearColorFromBool(
     Bool=$User.bIsFrontendPreview
     True Color=&ColorFromCurve(ColorCurve=DataInterface() {propertyValues {
        AlphaCurve {keys [
          {arriveTangentWeight=0.0086, interpMode=RCIM_Cubic, leaveTangentWeight=0.0086, tangentMode=RCTM_User, time=0.6, value=1}
          {tangentMode=RCTM_User, time=0.71, value=0}
         ]}
        BlueCurve {keys [
          {tangentMode=RCTM_User, time=0.16, value=0.08}
          {arriveTangentWeight=0.0655, interpMode=RCIM_Cubic, leaveTangentWeight=0.1021, tangentMode=RCTM_User, time=0.35, value=0.21}
          <at 0.5 0.21>
          {arriveTangentWeight=0.1021, interpMode=RCIM_Cubic, leaveTangentWeight=0.1021, tangentMode=RCTM_User, time=0.65, value=0.04}
         ]}
        GreenCurve {keys [
          {tangentMode=RCTM_User, time=0.16, value=0.08}
          {arriveTangentWeight=0.0655, interpMode=RCIM_Cubic, leaveTangentWeight=0.1021, tangentMode=RCTM_User, time=0.35, value=0.21}
          <at 0.5 0.21>
          {arriveTangentWeight=0.1021, interpMode=RCIM_Cubic, leaveTangentWeight=0.1021, tangentMode=RCTM_User, time=0.65, value=0.04}
         ]}
        RedCurve {keys [
          {tangentMode=RCTM_User, time=0.16, value=0.08}
          {arriveTangentWeight=0.0655, interpMode=RCIM_Cubic, leaveTangentWeight=0.1021, tangentMode=RCTM_User, time=0.35, value=0.21}
          <at 0.5 0.21>
          {arriveTangentWeight=0.1021, interpMode=RCIM_Cubic, leaveTangentWeight=0.1021, tangentMode=RCTM_User, time=0.65, value=0.04}
         ]}
       }})
     False Color=&ColorFromCurve(ColorCurve=DataInterface() {propertyValues {
        AlphaCurve {keys [
          {arriveTangentWeight=0.0086, interpMode=RCIM_Cubic, leaveTangentWeight=0.0086, tangentMode=RCTM_User, time=0.6, value=1}
          {tangentMode=RCTM_User, time=0.71, value=0}
         ]}
        BlueCurve {keys [
          {tangentMode=RCTM_User, time=0.1, value=0.08}
          {arriveTangentWeight=0.0655, interpMode=RCIM_Cubic, leaveTangentWeight=0.1021, tangentMode=RCTM_User, time=0.2, value=0.21}
          <at 0.5 0.21>
          {arriveTangentWeight=0.1021, interpMode=RCIM_Cubic, leaveTangentWeight=0.1021, tangentMode=RCTM_User, time=0.65, value=0.04}
         ]}
        GreenCurve {keys [
          {tangentMode=RCTM_User, time=0.1, value=0.08}
          {arriveTangentWeight=0.0655, interpMode=RCIM_Cubic, leaveTangentWeight=0.1021, tangentMode=RCTM_User, time=0.2, value=0.21}
          <at 0.5 0.21>
          {arriveTangentWeight=0.1021, interpMode=RCIM_Cubic, leaveTangentWeight=0.1021, tangentMode=RCTM_User, time=0.65, value=0.04}
         ]}
        RedCurve {keys [
          {tangentMode=RCTM_User, time=0.1, value=0.08}
          {arriveTangentWeight=0.0655, interpMode=RCIM_Cubic, leaveTangentWeight=0.1021, tangentMode=RCTM_User, time=0.2, value=0.21}
          <at 0.5 0.21>
          {arriveTangentWeight=0.1021, interpMode=RCIM_Cubic, leaveTangentWeight=0.1021, tangentMode=RCTM_User, time=0.65, value=0.04}
         ]}
       }})
    )
    R1 BurntSienna:LinearColor=&MakeCustomLinearColorFromBool(
     Bool=$User.bIsFrontendPreview
     True Color=&ColorFromCurve(ColorCurve=DataInterface() {propertyValues {
        AlphaCurve {keys [
          {arriveTangentWeight=0.0086, interpMode=RCIM_Cubic, leaveTangentWeight=0.0086, tangentMode=RCTM_User, time=0.6, value=1}
          {tangentMode=RCTM_User, time=0.71, value=0}
         ]}
        BlueCurve {keys [
          {tangentMode=RCTM_User, time=0.16, value=0.105}
          {arriveTangentWeight=0.0655, interpMode=RCIM_Cubic, leaveTangentWeight=0.1021, tangentMode=RCTM_User, time=0.35, value=0.2964}
          <at 0.5 0.2964>
          {arriveTangentWeight=0.1021, interpMode=RCIM_Cubic, leaveTangentWeight=0.1021, tangentMode=RCTM_User, time=0.65, value=0.0525}
         ]}
        GreenCurve {keys [
          {tangentMode=RCTM_User, time=0.16, value=0.2291}
          {arriveTangentWeight=0.0655, interpMode=RCIM_Cubic, leaveTangentWeight=0.1021, tangentMode=RCTM_User, time=0.35, value=0.6467}
          <at 0.5 0.6467>
          {arriveTangentWeight=0.1021, interpMode=RCIM_Cubic, leaveTangentWeight=0.1021, tangentMode=RCTM_User, time=0.65, value=0.1145}
         ]}
        RedCurve {keys [
          {tangentMode=RCTM_User, time=0.16, value=0.3}
          {arriveTangentWeight=0.0655, interpMode=RCIM_Cubic, leaveTangentWeight=0.1021, tangentMode=RCTM_User, time=0.35, value=0.8469}
          <at 0.5 0.8469>
          {arriveTangentWeight=0.1021, interpMode=RCIM_Cubic, leaveTangentWeight=0.1021, tangentMode=RCTM_User, time=0.65, value=0.15}
         ]}
       }})
     False Color=&ColorFromCurve(ColorCurve=DataInterface() {propertyValues {
        AlphaCurve {keys [
          {arriveTangentWeight=0.0086, interpMode=RCIM_Cubic, leaveTangentWeight=0.0086, tangentMode=RCTM_User, time=0.6, value=1}
          {tangentMode=RCTM_User, time=0.71, value=0}
         ]}
        BlueCurve {keys [
          {tangentMode=RCTM_User, time=0.1, value=0.105}
          {arriveTangentWeight=0.0655, interpMode=RCIM_Cubic, leaveTangentWeight=0.1021, tangentMode=RCTM_User, time=0.2, value=0.2964}
          <at 0.5 0.2964>
          {arriveTangentWeight=0.1021, interpMode=RCIM_Cubic, leaveTangentWeight=0.1021, tangentMode=RCTM_User, time=0.65, value=0.0525}
         ]}
        GreenCurve {keys [
          {tangentMode=RCTM_User, time=0.1, value=0.2291}
          {arriveTangentWeight=0.0655, interpMode=RCIM_Cubic, leaveTangentWeight=0.1021, tangentMode=RCTM_User, time=0.2, value=0.6467}
          <at 0.5 0.6467>
          {arriveTangentWeight=0.1021, interpMode=RCIM_Cubic, leaveTangentWeight=0.1021, tangentMode=RCTM_User, time=0.65, value=0.1145}
         ]}
        RedCurve {keys [
          {tangentMode=RCTM_User, time=0.1, value=0.3}
          {arriveTangentWeight=0.0655, interpMode=RCIM_Cubic, leaveTangentWeight=0.1021, tangentMode=RCTM_User, time=0.2, value=0.8469}
          <at 0.5 0.8469>
          {arriveTangentWeight=0.1021, interpMode=RCIM_Cubic, leaveTangentWeight=0.1021, tangentMode=RCTM_User, time=0.65, value=0.15}
         ]}
       }})
    )
    R1 Cobalt:LinearColor=&MakeCustomLinearColorFromBool(
     Bool=$User.bIsFrontendPreview
     True Color=&ColorFromCurve(ColorCurve=DataInterface() {propertyValues {
        AlphaCurve {keys [
          {arriveTangentWeight=0.0086, interpMode=RCIM_Cubic, leaveTangentWeight=0.0086, tangentMode=RCTM_User, time=0.6, value=1}
          {tangentMode=RCTM_User, time=0.71, value=0}
         ]}
        BlueCurve {keys [
          {tangentMode=RCTM_User, time=0.16, value=0.5972}
          {arriveTangentWeight=0.0655, interpMode=RCIM_Cubic, leaveTangentWeight=0.1021, tangentMode=RCTM_User, time=0.35, value=0.85}
          <at 0.5 0.85>
          {arriveTangentWeight=0.1021, interpMode=RCIM_Cubic, leaveTangentWeight=0.1021, tangentMode=RCTM_User, time=0.65, value=0.2448}
         ]}
        GreenCurve {keys [
          {tangentMode=RCTM_User, time=0.16, value=0.0666}
          {arriveTangentWeight=0.0655, interpMode=RCIM_Cubic, leaveTangentWeight=0.1021, tangentMode=RCTM_User, time=0.35, value=0.7225}
          <at 0.5 0.7225>
          {arriveTangentWeight=0.1021, interpMode=RCIM_Cubic, leaveTangentWeight=0.1021, tangentMode=RCTM_User, time=0.65, value=0.0269}
         ]}
        RedCurve {keys [
          {tangentMode=RCTM_User, time=0.16, value=0.0176}
          {arriveTangentWeight=0.0655, interpMode=RCIM_Cubic, leaveTangentWeight=0.1021, tangentMode=RCTM_User, time=0.35, value=0.4675}
          <at 0.5 0.4675>
          {arriveTangentWeight=0.1021, interpMode=RCIM_Cubic, leaveTangentWeight=0.1021, tangentMode=RCTM_User, time=0.65, value=0.0073}
         ]}
       }})
     False Color=&ColorFromCurve(ColorCurve=DataInterface() {propertyValues {
        AlphaCurve {keys [
          {arriveTangentWeight=0.0086, interpMode=RCIM_Cubic, leaveTangentWeight=0.0086, tangentMode=RCTM_User, time=0.6, value=1}
          {tangentMode=RCTM_User, time=0.71, value=0}
         ]}
        BlueCurve {keys [
          {tangentMode=RCTM_User, time=0.1, value=0.6}
          {arriveTangentWeight=0.0655, interpMode=RCIM_Cubic, leaveTangentWeight=0.1021, tangentMode=RCTM_User, time=0.2, value=0.8}
          <at 0.5 0.8>
          {arriveTangentWeight=0.1021, interpMode=RCIM_Cubic, leaveTangentWeight=0.1021, tangentMode=RCTM_User, time=0.65, value=0.4}
         ]}
        GreenCurve {keys [
          {tangentMode=RCTM_User, time=0.1, value=0.0775}
          {arriveTangentWeight=0.0655, interpMode=RCIM_Cubic, leaveTangentWeight=0.1021, tangentMode=RCTM_User, time=0.2, value=0.5667}
          <at 0.5 0.5667>
          {arriveTangentWeight=0.1021, interpMode=RCIM_Cubic, leaveTangentWeight=0.1021, tangentMode=RCTM_User, time=0.65, value=0.0517}
         ]}
        RedCurve {keys [
          {tangentMode=RCTM_User, time=0.1, value=0.03}
          {arriveTangentWeight=0.0655, interpMode=RCIM_Cubic, leaveTangentWeight=0.1021, tangentMode=RCTM_User, time=0.2, value=0.24}
          <at 0.5 0.24>
          {arriveTangentWeight=0.1021, interpMode=RCIM_Cubic, leaveTangentWeight=0.1021, tangentMode=RCTM_User, time=0.65, value=0.02}
         ]}
       }})
    )
    R1 Crimson:LinearColor=&MakeCustomLinearColorFromBool(
     Bool=$User.bIsFrontendPreview
     True Color=&ColorFromCurve(ColorCurve=DataInterface() {propertyValues {
        AlphaCurve {keys [
          {arriveTangentWeight=0.0086, interpMode=RCIM_Cubic, leaveTangentWeight=0.0086, tangentMode=RCTM_User, time=0.6, value=1}
          {tangentMode=RCTM_User, time=0.71, value=0}
         ]}
        BlueCurve {keys [
          {tangentMode=RCTM_User, time=0.16, value=0.0118}
          {arriveTangentWeight=0.0655, interpMode=RCIM_Cubic, leaveTangentWeight=0.1021, tangentMode=RCTM_User, time=0.35, value=0.495}
          <at 0.5 0.495>
          {arriveTangentWeight=0.1021, interpMode=RCIM_Cubic, leaveTangentWeight=0.1021, tangentMode=RCTM_User, time=0.65, value=0.009}
         ]}
        GreenCurve {keys [
          {tangentMode=RCTM_User, time=0.16, value=0.0183}
          {arriveTangentWeight=0.0655, interpMode=RCIM_Cubic, leaveTangentWeight=0.1021, tangentMode=RCTM_User, time=0.35, value=0.5017}
          <at 0.5 0.5017>
          {arriveTangentWeight=0.1021, interpMode=RCIM_Cubic, leaveTangentWeight=0.1021, tangentMode=RCTM_User, time=0.65, value=0.0139}
         ]}
        RedCurve {keys [
          {tangentMode=RCTM_User, time=0.16, value=0.4}
          {arriveTangentWeight=0.0655, interpMode=RCIM_Cubic, leaveTangentWeight=0.1021, tangentMode=RCTM_User, time=0.35, value=0.9}
          <at 0.5 0.9>
          {arriveTangentWeight=0.1021, interpMode=RCIM_Cubic, leaveTangentWeight=0.1021, tangentMode=RCTM_User, time=0.65, value=0.3}
         ]}
       }})
     False Color=&ColorFromCurve(ColorCurve=DataInterface() {propertyValues {
        AlphaCurve {keys [
          {arriveTangentWeight=0.0086, interpMode=RCIM_Cubic, leaveTangentWeight=0.0086, tangentMode=RCTM_User, time=0.6, value=1}
          {tangentMode=RCTM_User, time=0.71, value=0}
         ]}
        BlueCurve {keys [
          {tangentMode=RCTM_User, time=0.1, value=0.025}
          {arriveTangentWeight=0.0655, interpMode=RCIM_Cubic, leaveTangentWeight=0.1021, tangentMode=RCTM_User, time=0.2, value=0.495}
          <at 0.5 0.495>
          {arriveTangentWeight=0.1021, interpMode=RCIM_Cubic, leaveTangentWeight=0.1021, tangentMode=RCTM_User, time=0.65, value=0.02}
         ]}
        GreenCurve {keys [
          {tangentMode=RCTM_User, time=0.1, value=0.0329}
          {arriveTangentWeight=0.0655, interpMode=RCIM_Cubic, leaveTangentWeight=0.1021, tangentMode=RCTM_User, time=0.2, value=0.5017}
          <at 0.5 0.5017>
          {arriveTangentWeight=0.1021, interpMode=RCIM_Cubic, leaveTangentWeight=0.1021, tangentMode=RCTM_User, time=0.65, value=0.0263}
         ]}
        RedCurve {keys [
          {tangentMode=RCTM_User, time=0.1, value=0.5}
          {arriveTangentWeight=0.0655, interpMode=RCIM_Cubic, leaveTangentWeight=0.1021, tangentMode=RCTM_User, time=0.2, value=0.9}
          <at 0.5 0.9>
          {arriveTangentWeight=0.1021, interpMode=RCIM_Cubic, leaveTangentWeight=0.1021, tangentMode=RCTM_User, time=0.65, value=0.4}
         ]}
       }})
    )
    R1 DefaultColor=&MakeCustomLinearColorFromBool(
     Bool=$User.bIsFrontendPreview
     True Color=&ColorFromCurve(ColorCurve=DataInterface() {propertyValues {
        AlphaCurve {keys [
          {arriveTangentWeight=0.0086, interpMode=RCIM_Cubic, leaveTangentWeight=0.0086, tangentMode=RCTM_User, time=0.6, value=1}
          {tangentMode=RCTM_User, time=0.71, value=0}
         ]}
        BlueCurve {keys [
          {tangentMode=RCTM_User, time=0.16, value=0.5972}
          {arriveTangentWeight=0.0655, interpMode=RCIM_Cubic, leaveTangentWeight=0.1021, tangentMode=RCTM_User, time=0.35, value=1}
          <at 0.5 1>
          {arriveTangentWeight=0.1021, interpMode=RCIM_Cubic, leaveTangentWeight=0.1021, tangentMode=RCTM_User, time=0.65, value=0.2448}
         ]}
        GreenCurve {keys [
          {tangentMode=RCTM_User, time=0.16, value=0.0666}
          {arriveTangentWeight=0.0655, interpMode=RCIM_Cubic, leaveTangentWeight=0.1021, tangentMode=RCTM_User, time=0.35, value=0.9734}
          <at 0.5 0.9734>
          {arriveTangentWeight=0.1021, interpMode=RCIM_Cubic, leaveTangentWeight=0.1021, tangentMode=RCTM_User, time=0.65, value=0.0269}
         ]}
        RedCurve {keys [
          {tangentMode=RCTM_User, time=0.16, value=0.0176}
          {arriveTangentWeight=0.0655, interpMode=RCIM_Cubic, leaveTangentWeight=0.1021, tangentMode=RCTM_User, time=0.35, value=0.0482}
          <at 0.5 0.0482>
          {arriveTangentWeight=0.1021, interpMode=RCIM_Cubic, leaveTangentWeight=0.1021, tangentMode=RCTM_User, time=0.65, value=0.0073}
         ]}
       }})
     False Color=&ColorFromCurve(ColorCurve=DataInterface() {propertyValues {
        AlphaCurve {keys [
          {arriveTangentWeight=0.0086, interpMode=RCIM_Cubic, leaveTangentWeight=0.0086, tangentMode=RCTM_User, time=0.6, value=1}
          {tangentMode=RCTM_User, time=0.71, value=0}
         ]}
        BlueCurve {keys [
          {tangentMode=RCTM_User, time=0.1, value=0.5972}
          {arriveTangentWeight=0.0655, interpMode=RCIM_Cubic, leaveTangentWeight=0.1021, tangentMode=RCTM_User, time=0.2, value=1}
          <at 0.5 1>
          {arriveTangentWeight=0.1021, interpMode=RCIM_Cubic, leaveTangentWeight=0.1021, tangentMode=RCTM_User, time=0.65, value=0.2448}
         ]}
        GreenCurve {keys [
          {tangentMode=RCTM_User, time=0.1, value=0.0666}
          {arriveTangentWeight=0.0655, interpMode=RCIM_Cubic, leaveTangentWeight=0.1021, tangentMode=RCTM_User, time=0.2, value=0.9734}
          <at 0.5 0.9734>
          {arriveTangentWeight=0.1021, interpMode=RCIM_Cubic, leaveTangentWeight=0.1021, tangentMode=RCTM_User, time=0.65, value=0.0269}
         ]}
        RedCurve {keys [
          {tangentMode=RCTM_User, time=0.1, value=0.0176}
          {arriveTangentWeight=0.0655, interpMode=RCIM_Cubic, leaveTangentWeight=0.1021, tangentMode=RCTM_User, time=0.2, value=0.0482}
          <at 0.5 0.0482>
          {arriveTangentWeight=0.1021, interpMode=RCIM_Cubic, leaveTangentWeight=0.1021, tangentMode=RCTM_User, time=0.65, value=0.0073}
         ]}
       }})
    )
    R1 ForestGreen:LinearColor=&MakeCustomLinearColorFromBool(
     Bool=$User.bIsFrontendPreview
     True Color=&ColorFromCurve(ColorCurve=DataInterface() {propertyValues {
        AlphaCurve {keys [
          {arriveTangentWeight=0.0086, interpMode=RCIM_Cubic, leaveTangentWeight=0.0086, tangentMode=RCTM_User, time=0.6, value=1}
          {tangentMode=RCTM_User, time=0.71, value=0}
         ]}
        BlueCurve {keys [
          {tangentMode=RCTM_User, time=0.16, value=0.07}
          {arriveTangentWeight=0.0655, interpMode=RCIM_Cubic, leaveTangentWeight=0.1021, tangentMode=RCTM_User, time=0.35, value=0.2975}
          <at 0.5 0.2975>
          {arriveTangentWeight=0.1021, interpMode=RCIM_Cubic, leaveTangentWeight=0.1021, tangentMode=RCTM_User, time=0.65, value=0.035}
         ]}
        GreenCurve {keys [
          {tangentMode=RCTM_User, time=0.16, value=0.2}
          {arriveTangentWeight=0.0655, interpMode=RCIM_Cubic, leaveTangentWeight=0.1021, tangentMode=RCTM_User, time=0.35, value=0.85}
          <at 0.5 0.85>
          {arriveTangentWeight=0.1021, interpMode=RCIM_Cubic, leaveTangentWeight=0.1021, tangentMode=RCTM_User, time=0.65, value=0.1}
         ]}
        RedCurve {keys [
          {tangentMode=RCTM_User, time=0.16, value=0.07}
          {arriveTangentWeight=0.0655, interpMode=RCIM_Cubic, leaveTangentWeight=0.1021, tangentMode=RCTM_User, time=0.35, value=0.2975}
          <at 0.5 0.2975>
          {arriveTangentWeight=0.1021, interpMode=RCIM_Cubic, leaveTangentWeight=0.1021, tangentMode=RCTM_User, time=0.65, value=0.035}
         ]}
       }})
     False Color=&ColorFromCurve(ColorCurve=DataInterface() {propertyValues {
        AlphaCurve {keys [
          {arriveTangentWeight=0.0086, interpMode=RCIM_Cubic, leaveTangentWeight=0.0086, tangentMode=RCTM_User, time=0.6, value=1}
          {tangentMode=RCTM_User, time=0.71, value=0}
         ]}
        BlueCurve {keys [
          {tangentMode=RCTM_User, time=0.1, value=0.07}
          {arriveTangentWeight=0.0655, interpMode=RCIM_Cubic, leaveTangentWeight=0.1021, tangentMode=RCTM_User, time=0.2, value=0.255}
          <at 0.5 0.255>
          {arriveTangentWeight=0.1021, interpMode=RCIM_Cubic, leaveTangentWeight=0.1021, tangentMode=RCTM_User, time=0.65, value=0.042}
         ]}
        GreenCurve {keys [
          {tangentMode=RCTM_User, time=0.1, value=0.2}
          {arriveTangentWeight=0.0655, interpMode=RCIM_Cubic, leaveTangentWeight=0.1021, tangentMode=RCTM_User, time=0.2, value=0.85}
          <at 0.5 0.85>
          {arriveTangentWeight=0.1021, interpMode=RCIM_Cubic, leaveTangentWeight=0.1021, tangentMode=RCTM_User, time=0.65, value=0.12}
         ]}
        RedCurve {keys [
          {tangentMode=RCTM_User, time=0.1, value=0.07}
          {arriveTangentWeight=0.0655, interpMode=RCIM_Cubic, leaveTangentWeight=0.1021, tangentMode=RCTM_User, time=0.2, value=0.255}
          <at 0.5 0.255>
          {arriveTangentWeight=0.1021, interpMode=RCIM_Cubic, leaveTangentWeight=0.1021, tangentMode=RCTM_User, time=0.65, value=0.042}
         ]}
       }})
    )
    R1 Grey:LinearColor=&MakeCustomLinearColorFromBool(
     Bool=$User.bIsFrontendPreview
     True Color=&ColorFromCurve(ColorCurve=DataInterface() {propertyValues {
        AlphaCurve {keys [
          {arriveTangentWeight=0.0086, interpMode=RCIM_Cubic, leaveTangentWeight=0.0086, tangentMode=RCTM_User, time=0.6, value=1}
          {tangentMode=RCTM_User, time=0.71, value=0}
         ]}
        BlueCurve {keys [
          {tangentMode=RCTM_User, time=0.16, value=0.08}
          {arriveTangentWeight=0.0655, interpMode=RCIM_Cubic, leaveTangentWeight=0.1021, tangentMode=RCTM_User, time=0.35, value=0.25}
          <at 0.5 0.25>
          {arriveTangentWeight=0.1021, interpMode=RCIM_Cubic, leaveTangentWeight=0.1021, tangentMode=RCTM_User, time=0.65, value=0.05}
         ]}
        GreenCurve {keys [
          {tangentMode=RCTM_User, time=0.16, value=0.08}
          {arriveTangentWeight=0.0655, interpMode=RCIM_Cubic, leaveTangentWeight=0.1021, tangentMode=RCTM_User, time=0.35, value=0.25}
          <at 0.5 0.25>
          {arriveTangentWeight=0.1021, interpMode=RCIM_Cubic, leaveTangentWeight=0.1021, tangentMode=RCTM_User, time=0.65, value=0.05}
         ]}
        RedCurve {keys [
          {tangentMode=RCTM_User, time=0.16, value=0.08}
          {arriveTangentWeight=0.0655, interpMode=RCIM_Cubic, leaveTangentWeight=0.1021, tangentMode=RCTM_User, time=0.35, value=0.25}
          <at 0.5 0.25>
          {arriveTangentWeight=0.1021, interpMode=RCIM_Cubic, leaveTangentWeight=0.1021, tangentMode=RCTM_User, time=0.65, value=0.05}
         ]}
       }})
     False Color=&ColorFromCurve(ColorCurve=DataInterface() {propertyValues {
        AlphaCurve {keys [
          {arriveTangentWeight=0.0086, interpMode=RCIM_Cubic, leaveTangentWeight=0.0086, tangentMode=RCTM_User, time=0.6, value=1}
          {tangentMode=RCTM_User, time=0.71, value=0}
         ]}
        BlueCurve {keys [
          {tangentMode=RCTM_User, time=0.1, value=0.1}
          {arriveTangentWeight=0.0655, interpMode=RCIM_Cubic, leaveTangentWeight=0.1021, tangentMode=RCTM_User, time=0.2, value=0.5}
          <at 0.5 0.5>
          {arriveTangentWeight=0.1021, interpMode=RCIM_Cubic, leaveTangentWeight=0.1021, tangentMode=RCTM_User, time=0.65, value=0.12}
         ]}
        GreenCurve {keys [
          {tangentMode=RCTM_User, time=0.1, value=0.1}
          {arriveTangentWeight=0.0655, interpMode=RCIM_Cubic, leaveTangentWeight=0.1021, tangentMode=RCTM_User, time=0.2, value=0.5}
          <at 0.5 0.5>
          {arriveTangentWeight=0.1021, interpMode=RCIM_Cubic, leaveTangentWeight=0.1021, tangentMode=RCTM_User, time=0.65, value=0.12}
         ]}
        RedCurve {keys [
          {tangentMode=RCTM_User, time=0.1, value=0.1}
          {arriveTangentWeight=0.0655, interpMode=RCIM_Cubic, leaveTangentWeight=0.1021, tangentMode=RCTM_User, time=0.2, value=0.5}
          <at 0.5 0.5>
          {arriveTangentWeight=0.1021, interpMode=RCIM_Cubic, leaveTangentWeight=0.1021, tangentMode=RCTM_User, time=0.65, value=0.12}
         ]}
       }})
    )
    R1 Lime:LinearColor=&MakeCustomLinearColorFromBool(
     Bool=$User.bIsFrontendPreview
     True Color=&ColorFromCurve(ColorCurve=DataInterface() {propertyValues {
        AlphaCurve {keys [
          {arriveTangentWeight=0.0086, interpMode=RCIM_Cubic, leaveTangentWeight=0.0086, tangentMode=RCTM_User, time=0.6, value=1}
          {tangentMode=RCTM_User, time=0.71, value=0}
         ]}
        BlueCurve {keys [
          {tangentMode=RCTM_User, time=0.16, value=0.0118}
          {arriveTangentWeight=0.0655, interpMode=RCIM_Cubic, leaveTangentWeight=0.1021, tangentMode=RCTM_User, time=0.35, value=0.024}
          <at 0.5 0.024>
          {arriveTangentWeight=0.1021, interpMode=RCIM_Cubic, leaveTangentWeight=0.1021, tangentMode=RCTM_User, time=0.65, value=0.006}
         ]}
        GreenCurve {keys [
          {tangentMode=RCTM_User, time=0.16, value=0.4}
          {arriveTangentWeight=0.0655, interpMode=RCIM_Cubic, leaveTangentWeight=0.1021, tangentMode=RCTM_User, time=0.35, value=0.8}
          <at 0.5 0.8>
          {arriveTangentWeight=0.1021, interpMode=RCIM_Cubic, leaveTangentWeight=0.1021, tangentMode=RCTM_User, time=0.65, value=0.2}
         ]}
        RedCurve {keys [
          {tangentMode=RCTM_User, time=0.16, value=0.3353}
          {arriveTangentWeight=0.0655, interpMode=RCIM_Cubic, leaveTangentWeight=0.1021, tangentMode=RCTM_User, time=0.35, value=0.6707}
          <at 0.5 0.6707>
          {arriveTangentWeight=0.1021, interpMode=RCIM_Cubic, leaveTangentWeight=0.1021, tangentMode=RCTM_User, time=0.65, value=0.1677}
         ]}
       }})
     False Color=&ColorFromCurve(ColorCurve=DataInterface() {propertyValues {
        AlphaCurve {keys [
          {arriveTangentWeight=0.0086, interpMode=RCIM_Cubic, leaveTangentWeight=0.0086, tangentMode=RCTM_User, time=0.6, value=1}
          {tangentMode=RCTM_User, time=0.71, value=0}
         ]}
        BlueCurve {keys [
          {tangentMode=RCTM_User, time=0.1, value=0.0118}
          {arriveTangentWeight=0.0655, interpMode=RCIM_Cubic, leaveTangentWeight=0.1021, tangentMode=RCTM_User, time=0.2, value=0.0255}
          <at 0.5 0.0255>
          {arriveTangentWeight=0.1021, interpMode=RCIM_Cubic, leaveTangentWeight=0.1021, tangentMode=RCTM_User, time=0.65, value=0.006}
         ]}
        GreenCurve {keys [
          {tangentMode=RCTM_User, time=0.1, value=0.4}
          {arriveTangentWeight=0.0655, interpMode=RCIM_Cubic, leaveTangentWeight=0.1021, tangentMode=RCTM_User, time=0.2, value=0.85}
          <at 0.5 0.85>
          {arriveTangentWeight=0.1021, interpMode=RCIM_Cubic, leaveTangentWeight=0.1021, tangentMode=RCTM_User, time=0.65, value=0.2}
         ]}
        RedCurve {keys [
          {tangentMode=RCTM_User, time=0.1, value=0.3353}
          {arriveTangentWeight=0.0655, interpMode=RCIM_Cubic, leaveTangentWeight=0.1021, tangentMode=RCTM_User, time=0.2, value=0.7126}
          <at 0.5 0.7126>
          {arriveTangentWeight=0.1021, interpMode=RCIM_Cubic, leaveTangentWeight=0.1021, tangentMode=RCTM_User, time=0.65, value=0.1677}
         ]}
       }})
    )
    R1 Orange:LinearColor=&MakeCustomLinearColorFromBool(
     Bool=$User.bIsFrontendPreview
     True Color=&ColorFromCurve(ColorCurve=DataInterface() {propertyValues {
        AlphaCurve {keys [
          {arriveTangentWeight=0.0086, interpMode=RCIM_Cubic, leaveTangentWeight=0.0086, tangentMode=RCTM_User, time=0.6, value=1}
          {tangentMode=RCTM_User, time=0.71, value=0}
         ]}
        BlueCurve {keys [
          {tangentMode=RCTM_User, time=0.16, value=0.006}
          {arriveTangentWeight=0.0655, interpMode=RCIM_Cubic, leaveTangentWeight=0.1021, tangentMode=RCTM_User, time=0.35, value=0.017}
          <at 0.5 0.017>
          {arriveTangentWeight=0.1021, interpMode=RCIM_Cubic, leaveTangentWeight=0.1021, tangentMode=RCTM_User, time=0.65, value=0.003}
         ]}
        GreenCurve {keys [
          {tangentMode=RCTM_User, time=0.16, value=0.202}
          {arriveTangentWeight=0.0655, interpMode=RCIM_Cubic, leaveTangentWeight=0.1021, tangentMode=RCTM_User, time=0.35, value=0.5723}
          <at 0.5 0.5723>
          {arriveTangentWeight=0.1021, interpMode=RCIM_Cubic, leaveTangentWeight=0.1021, tangentMode=RCTM_User, time=0.65, value=0.101}
         ]}
        RedCurve {keys [
          {tangentMode=RCTM_User, time=0.16, value=0.3}
          {arriveTangentWeight=0.0655, interpMode=RCIM_Cubic, leaveTangentWeight=0.1021, tangentMode=RCTM_User, time=0.35, value=0.85}
          <at 0.5 0.85>
          {arriveTangentWeight=0.1021, interpMode=RCIM_Cubic, leaveTangentWeight=0.1021, tangentMode=RCTM_User, time=0.65, value=0.15}
         ]}
       }})
     False Color=&ColorFromCurve(ColorCurve=DataInterface() {propertyValues {
        AlphaCurve {keys [
          {arriveTangentWeight=0.0086, interpMode=RCIM_Cubic, leaveTangentWeight=0.0086, tangentMode=RCTM_User, time=0.6, value=1}
          {tangentMode=RCTM_User, time=0.71, value=0}
         ]}
        BlueCurve {keys [
          {tangentMode=RCTM_User, time=0.1, value=0.006}
          {arriveTangentWeight=0.0655, interpMode=RCIM_Cubic, leaveTangentWeight=0.1021, tangentMode=RCTM_User, time=0.2, value=0.017}
          <at 0.5 0.017>
          {arriveTangentWeight=0.1021, interpMode=RCIM_Cubic, leaveTangentWeight=0.1021, tangentMode=RCTM_User, time=0.65, value=0.003}
         ]}
        GreenCurve {keys [
          {tangentMode=RCTM_User, time=0.1, value=0.202}
          {arriveTangentWeight=0.0655, interpMode=RCIM_Cubic, leaveTangentWeight=0.1021, tangentMode=RCTM_User, time=0.2, value=0.5723}
          <at 0.5 0.5723>
          {arriveTangentWeight=0.1021, interpMode=RCIM_Cubic, leaveTangentWeight=0.1021, tangentMode=RCTM_User, time=0.65, value=0.101}
         ]}
        RedCurve {keys [
          {tangentMode=RCTM_User, time=0.1, value=0.3}
          {arriveTangentWeight=0.0655, interpMode=RCIM_Cubic, leaveTangentWeight=0.1021, tangentMode=RCTM_User, time=0.2, value=0.85}
          <at 0.5 0.85>
          {arriveTangentWeight=0.1021, interpMode=RCIM_Cubic, leaveTangentWeight=0.1021, tangentMode=RCTM_User, time=0.65, value=0.15}
         ]}
       }})
    )
    R1 Pink:LinearColor=&MakeCustomLinearColorFromBool(
     Bool=$User.bIsFrontendPreview
     True Color=&ColorFromCurve(ColorCurve=DataInterface() {propertyValues {
        AlphaCurve {keys [
          {arriveTangentWeight=0.0086, interpMode=RCIM_Cubic, leaveTangentWeight=0.0086, tangentMode=RCTM_User, time=0.6, value=1}
          {tangentMode=RCTM_User, time=0.71, value=0}
         ]}
        BlueCurve {keys [
          {tangentMode=RCTM_User, time=0.16, value=0.3}
          {arriveTangentWeight=0.0655, interpMode=RCIM_Cubic, leaveTangentWeight=0.1021, tangentMode=RCTM_User, time=0.35, value=0.8}
          <at 0.5 0.8>
          {arriveTangentWeight=0.1021, interpMode=RCIM_Cubic, leaveTangentWeight=0.1021, tangentMode=RCTM_User, time=0.65, value=0.1}
         ]}
        GreenCurve {keys [
          {tangentMode=RCTM_User, time=0.16, value=0.075}
          {arriveTangentWeight=0.0655, interpMode=RCIM_Cubic, leaveTangentWeight=0.1021, tangentMode=RCTM_User, time=0.35, value=0.4}
          <at 0.5 0.4>
          {arriveTangentWeight=0.1021, interpMode=RCIM_Cubic, leaveTangentWeight=0.1021, tangentMode=RCTM_User, time=0.65, value=0.025}
         ]}
        RedCurve {keys [
          {tangentMode=RCTM_User, time=0.16, value=0.2438}
          {arriveTangentWeight=0.0655, interpMode=RCIM_Cubic, leaveTangentWeight=0.1021, tangentMode=RCTM_User, time=0.35, value=0.7}
          <at 0.5 0.7>
          {arriveTangentWeight=0.1021, interpMode=RCIM_Cubic, leaveTangentWeight=0.1021, tangentMode=RCTM_User, time=0.65, value=0.0812}
         ]}
       }})
     False Color=&ColorFromCurve(ColorCurve=DataInterface() {propertyValues {
        AlphaCurve {keys [
          {arriveTangentWeight=0.0086, interpMode=RCIM_Cubic, leaveTangentWeight=0.0086, tangentMode=RCTM_User, time=0.6, value=1}
          {tangentMode=RCTM_User, time=0.71, value=0}
         ]}
        BlueCurve {keys [
          {tangentMode=RCTM_User, time=0.1, value=0.4}
          {arriveTangentWeight=0.0655, interpMode=RCIM_Cubic, leaveTangentWeight=0.1021, tangentMode=RCTM_User, time=0.2, value=0.9}
          <at 0.5 0.9>
          {arriveTangentWeight=0.1021, interpMode=RCIM_Cubic, leaveTangentWeight=0.1021, tangentMode=RCTM_User, time=0.65, value=0.2}
         ]}
        GreenCurve {keys [
          {tangentMode=RCTM_User, time=0.1, value=0.1}
          {arriveTangentWeight=0.0655, interpMode=RCIM_Cubic, leaveTangentWeight=0.1021, tangentMode=RCTM_User, time=0.2, value=0.45}
          <at 0.5 0.45>
          {arriveTangentWeight=0.1021, interpMode=RCIM_Cubic, leaveTangentWeight=0.1021, tangentMode=RCTM_User, time=0.65, value=0.05}
         ]}
        RedCurve {keys [
          {tangentMode=RCTM_User, time=0.1, value=0.325}
          {arriveTangentWeight=0.0655, interpMode=RCIM_Cubic, leaveTangentWeight=0.1021, tangentMode=RCTM_User, time=0.2, value=0.7875}
          <at 0.5 0.7875>
          {arriveTangentWeight=0.1021, interpMode=RCIM_Cubic, leaveTangentWeight=0.1021, tangentMode=RCTM_User, time=0.65, value=0.1625}
         ]}
       }})
    )
    R1 Purple:LinearColor=&MakeCustomLinearColorFromBool(
     Bool=$User.bIsFrontendPreview
     True Color=&ColorFromCurve(ColorCurve=DataInterface() {propertyValues {
        AlphaCurve {keys [
          {arriveTangentWeight=0.0086, interpMode=RCIM_Cubic, leaveTangentWeight=0.0086, tangentMode=RCTM_User, time=0.6, value=1}
          {tangentMode=RCTM_User, time=0.71, value=0}
         ]}
        BlueCurve {keys [
          {tangentMode=RCTM_User, time=0.16, value=0.2}
          {arriveTangentWeight=0.0655, interpMode=RCIM_Cubic, leaveTangentWeight=0.1021, tangentMode=RCTM_User, time=0.35, value=0.9}
          <at 0.5 0.9>
          {arriveTangentWeight=0.1021, interpMode=RCIM_Cubic, leaveTangentWeight=0.1021, tangentMode=RCTM_User, time=0.65, value=0.08}
         ]}
        GreenCurve {keys [
          {tangentMode=RCTM_User, time=0.16, value=0.05}
          {arriveTangentWeight=0.0655, interpMode=RCIM_Cubic, leaveTangentWeight=0.1021, tangentMode=RCTM_User, time=0.35, value=0.315}
          <at 0.5 0.315>
          {arriveTangentWeight=0.1021, interpMode=RCIM_Cubic, leaveTangentWeight=0.1021, tangentMode=RCTM_User, time=0.65, value=0.02}
         ]}
        RedCurve {keys [
          {tangentMode=RCTM_User, time=0.16, value=0.105}
          {arriveTangentWeight=0.0655, interpMode=RCIM_Cubic, leaveTangentWeight=0.1021, tangentMode=RCTM_User, time=0.35, value=0.705}
          <at 0.5 0.705>
          {arriveTangentWeight=0.1021, interpMode=RCIM_Cubic, leaveTangentWeight=0.1021, tangentMode=RCTM_User, time=0.65, value=0.042}
         ]}
       }})
     False Color=&ColorFromCurve(ColorCurve=DataInterface() {propertyValues {
        AlphaCurve {keys [
          {arriveTangentWeight=0.0086, interpMode=RCIM_Cubic, leaveTangentWeight=0.0086, tangentMode=RCTM_User, time=0.6, value=1}
          {tangentMode=RCTM_User, time=0.71, value=0}
         ]}
        BlueCurve {keys [
          {tangentMode=RCTM_User, time=0.1, value=0.3}
          {arriveTangentWeight=0.0655, interpMode=RCIM_Cubic, leaveTangentWeight=0.1021, tangentMode=RCTM_User, time=0.2, value=0.95}
          <at 0.5 0.95>
          {arriveTangentWeight=0.1021, interpMode=RCIM_Cubic, leaveTangentWeight=0.1021, tangentMode=RCTM_User, time=0.65, value=0.25}
         ]}
        GreenCurve {keys [
          {tangentMode=RCTM_User, time=0.1, value=0.075}
          {arriveTangentWeight=0.0655, interpMode=RCIM_Cubic, leaveTangentWeight=0.1021, tangentMode=RCTM_User, time=0.2, value=0.3325}
          <at 0.5 0.3325>
          {arriveTangentWeight=0.1021, interpMode=RCIM_Cubic, leaveTangentWeight=0.1021, tangentMode=RCTM_User, time=0.65, value=0.0625}
         ]}
        RedCurve {keys [
          {tangentMode=RCTM_User, time=0.1, value=0.1575}
          {arriveTangentWeight=0.0655, interpMode=RCIM_Cubic, leaveTangentWeight=0.1021, tangentMode=RCTM_User, time=0.2, value=0.7442}
          <at 0.5 0.7442>
          {arriveTangentWeight=0.1021, interpMode=RCIM_Cubic, leaveTangentWeight=0.1021, tangentMode=RCTM_User, time=0.65, value=0.1313}
         ]}
       }})
    )
    R1 Saffron:LinearColor=&MakeCustomLinearColorFromBool(
     Bool=$User.bIsFrontendPreview
     True Color=&ColorFromCurve(ColorCurve=DataInterface() {propertyValues {
        AlphaCurve {keys [
          {arriveTangentWeight=0.0086, interpMode=RCIM_Cubic, leaveTangentWeight=0.0086, tangentMode=RCTM_User, time=0.6, value=1}
          {tangentMode=RCTM_User, time=0.71, value=0}
         ]}
        BlueCurve {keys [
          {tangentMode=RCTM_User, time=0.16, value=0.0118}
          {arriveTangentWeight=0.0655, interpMode=RCIM_Cubic, leaveTangentWeight=0.1021, tangentMode=RCTM_User, time=0.35, value=0.024}
          <at 0.5 0.024>
          {arriveTangentWeight=0.1021, interpMode=RCIM_Cubic, leaveTangentWeight=0.1021, tangentMode=RCTM_User, time=0.65, value=0.006}
         ]}
        GreenCurve {keys [
          {tangentMode=RCTM_User, time=0.16, value=0.4}
          {arriveTangentWeight=0.0655, interpMode=RCIM_Cubic, leaveTangentWeight=0.1021, tangentMode=RCTM_User, time=0.35, value=0.8}
          <at 0.5 0.8>
          {arriveTangentWeight=0.1021, interpMode=RCIM_Cubic, leaveTangentWeight=0.1021, tangentMode=RCTM_User, time=0.65, value=0.2}
         ]}
        RedCurve {keys [
          {tangentMode=RCTM_User, time=0.16, value=0.4}
          {arriveTangentWeight=0.0655, interpMode=RCIM_Cubic, leaveTangentWeight=0.1021, tangentMode=RCTM_User, time=0.35, value=0.8}
          <at 0.5 0.8>
          {arriveTangentWeight=0.1021, interpMode=RCIM_Cubic, leaveTangentWeight=0.1021, tangentMode=RCTM_User, time=0.65, value=0.2}
         ]}
       }})
     False Color=&ColorFromCurve(ColorCurve=DataInterface() {propertyValues {
        AlphaCurve {keys [
          {arriveTangentWeight=0.0086, interpMode=RCIM_Cubic, leaveTangentWeight=0.0086, tangentMode=RCTM_User, time=0.6, value=1}
          {tangentMode=RCTM_User, time=0.71, value=0}
         ]}
        BlueCurve {keys [
          {tangentMode=RCTM_User, time=0.1, value=0.0118}
          {arriveTangentWeight=0.0655, interpMode=RCIM_Cubic, leaveTangentWeight=0.1021, tangentMode=RCTM_User, time=0.2, value=0.0255}
          <at 0.5 0.0255>
          {arriveTangentWeight=0.1021, interpMode=RCIM_Cubic, leaveTangentWeight=0.1021, tangentMode=RCTM_User, time=0.65, value=0.0075}
         ]}
        GreenCurve {keys [
          {tangentMode=RCTM_User, time=0.1, value=0.4}
          {arriveTangentWeight=0.0655, interpMode=RCIM_Cubic, leaveTangentWeight=0.1021, tangentMode=RCTM_User, time=0.2, value=0.85}
          <at 0.5 0.85>
          {arriveTangentWeight=0.1021, interpMode=RCIM_Cubic, leaveTangentWeight=0.1021, tangentMode=RCTM_User, time=0.65, value=0.25}
         ]}
        RedCurve {keys [
          {tangentMode=RCTM_User, time=0.1, value=0.4}
          {arriveTangentWeight=0.0655, interpMode=RCIM_Cubic, leaveTangentWeight=0.1021, tangentMode=RCTM_User, time=0.2, value=0.85}
          <at 0.5 0.85>
          {arriveTangentWeight=0.1021, interpMode=RCIM_Cubic, leaveTangentWeight=0.1021, tangentMode=RCTM_User, time=0.65, value=0.25}
         ]}
       }})
    )
    R1 SkyBlue:LinearColor=&MakeCustomLinearColorFromBool(
     Bool=$User.bIsFrontendPreview
     True Color=&ColorFromCurve(ColorCurve=DataInterface() {propertyValues {
        AlphaCurve {keys [
          {arriveTangentWeight=0.0086, interpMode=RCIM_Cubic, leaveTangentWeight=0.0086, tangentMode=RCTM_User, time=0.6, value=1}
          {tangentMode=RCTM_User, time=0.71, value=0}
         ]}
        BlueCurve {keys [
          {tangentMode=RCTM_User, time=0.16, value=0.4}
          {arriveTangentWeight=0.0655, interpMode=RCIM_Cubic, leaveTangentWeight=0.1021, tangentMode=RCTM_User, time=0.35, value=0.9}
          <at 0.5 0.9>
          {arriveTangentWeight=0.1021, interpMode=RCIM_Cubic, leaveTangentWeight=0.1021, tangentMode=RCTM_User, time=0.65, value=0.2}
         ]}
        GreenCurve {keys [
          {tangentMode=RCTM_User, time=0.16, value=0.1517}
          {arriveTangentWeight=0.0655, interpMode=RCIM_Cubic, leaveTangentWeight=0.1021, tangentMode=RCTM_User, time=0.35, value=0.3414}
          <at 0.5 0.312>
          {arriveTangentWeight=0.1021, interpMode=RCIM_Cubic, leaveTangentWeight=0.1021, tangentMode=RCTM_User, time=0.65, value=0.0733}
         ]}
        RedCurve {keys [
          {tangentMode=RCTM_User, time=0.16, value=0.008}
          {arriveTangentWeight=0.0655, interpMode=RCIM_Cubic, leaveTangentWeight=0.1021, tangentMode=RCTM_User, time=0.35, value=0.018}
          <at 0.5 0.018>
          {arriveTangentWeight=0.1021, interpMode=RCIM_Cubic, leaveTangentWeight=0.1021, tangentMode=RCTM_User, time=0.65, value=0}
         ]}
       }})
     False Color=&ColorFromCurve(ColorCurve=DataInterface() {propertyValues {
        AlphaCurve {keys [
          {arriveTangentWeight=0.0086, interpMode=RCIM_Cubic, leaveTangentWeight=0.0086, tangentMode=RCTM_User, time=0.6, value=1}
          {tangentMode=RCTM_User, time=0.71, value=0}
         ]}
        BlueCurve {keys [
          {tangentMode=RCTM_User, time=0.1, value=0.5}
          {arriveTangentWeight=0.0655, interpMode=RCIM_Cubic, leaveTangentWeight=0.1021, tangentMode=RCTM_User, time=0.2, value=0.9}
          <at 0.5 0.9>
          {arriveTangentWeight=0.1021, interpMode=RCIM_Cubic, leaveTangentWeight=0.1021, tangentMode=RCTM_User, time=0.65, value=0.3}
         ]}
        GreenCurve {keys [
          {tangentMode=RCTM_User, time=0.1, value=0.3367}
          {arriveTangentWeight=0.0655, interpMode=RCIM_Cubic, leaveTangentWeight=0.1021, tangentMode=RCTM_User, time=0.2, value=0.606}
          <at 0.5 0.606>
          {arriveTangentWeight=0.1021, interpMode=RCIM_Cubic, leaveTangentWeight=0.1021, tangentMode=RCTM_User, time=0.65, value=0.2}
         ]}
        RedCurve {keys [
          {tangentMode=RCTM_User, time=0.1, value=0.01}
          {arriveTangentWeight=0.0655, interpMode=RCIM_Cubic, leaveTangentWeight=0.1021, tangentMode=RCTM_User, time=0.2, value=0.018}
          <at 0.5 0.018>
          {arriveTangentWeight=0.1021, interpMode=RCIM_Cubic, leaveTangentWeight=0.1021, tangentMode=RCTM_User, time=0.65, value=0}
         ]}
       }})
    )
    R1 TitaniumWhite:LinearColor=&MakeCustomLinearColorFromBool(
     Bool=$User.bIsFrontendPreview
     True Color=&ColorFromCurve(ColorCurve=DataInterface() {propertyValues {
        AlphaCurve {keys [
          {arriveTangentWeight=0.0086, interpMode=RCIM_Cubic, leaveTangentWeight=0.0086, tangentMode=RCTM_User, time=0.6, value=1}
          {tangentMode=RCTM_User, time=0.71, value=0}
         ]}
        BlueCurve {keys [
          {tangentMode=RCTM_User, time=0.16, value=0.1}
          {arriveTangentWeight=0.0655, interpMode=RCIM_Cubic, leaveTangentWeight=0.1021, tangentMode=RCTM_User, time=0.35, value=0.8}
          <at 0.5 0.8>
          {arriveTangentWeight=0.1021, interpMode=RCIM_Cubic, leaveTangentWeight=0.1021, tangentMode=RCTM_User, time=0.65, value=0.2}
         ]}
        GreenCurve {keys [
          {tangentMode=RCTM_User, time=0.16, value=0.1}
          {arriveTangentWeight=0.0655, interpMode=RCIM_Cubic, leaveTangentWeight=0.1021, tangentMode=RCTM_User, time=0.35, value=0.8}
          <at 0.5 0.8>
          {arriveTangentWeight=0.1021, interpMode=RCIM_Cubic, leaveTangentWeight=0.1021, tangentMode=RCTM_User, time=0.65, value=0.2}
         ]}
        RedCurve {keys [
          {tangentMode=RCTM_User, time=0.16, value=0.1}
          {arriveTangentWeight=0.0655, interpMode=RCIM_Cubic, leaveTangentWeight=0.1021, tangentMode=RCTM_User, time=0.35, value=0.8}
          <at 0.5 0.8>
          {arriveTangentWeight=0.1021, interpMode=RCIM_Cubic, leaveTangentWeight=0.1021, tangentMode=RCTM_User, time=0.65, value=0.2}
         ]}
       }})
     False Color=&ColorFromCurve(ColorCurve=DataInterface() {propertyValues {
        AlphaCurve {keys [
          {arriveTangentWeight=0.0086, interpMode=RCIM_Cubic, leaveTangentWeight=0.0086, tangentMode=RCTM_User, time=0.6, value=1}
          {tangentMode=RCTM_User, time=0.71, value=0}
         ]}
        BlueCurve {keys [
          {tangentMode=RCTM_User, time=0.1, value=0.15}
          {arriveTangentWeight=0.0655, interpMode=RCIM_Cubic, leaveTangentWeight=0.1021, tangentMode=RCTM_User, time=0.2, value=0.9}
          <at 0.5 0.9>
          {arriveTangentWeight=0.1021, interpMode=RCIM_Cubic, leaveTangentWeight=0.1021, tangentMode=RCTM_User, time=0.65, value=0.2}
         ]}
        GreenCurve {keys [
          {tangentMode=RCTM_User, time=0.1, value=0.15}
          {arriveTangentWeight=0.0655, interpMode=RCIM_Cubic, leaveTangentWeight=0.1021, tangentMode=RCTM_User, time=0.2, value=0.9}
          <at 0.5 0.9>
          {arriveTangentWeight=0.1021, interpMode=RCIM_Cubic, leaveTangentWeight=0.1021, tangentMode=RCTM_User, time=0.65, value=0.2}
         ]}
        RedCurve {keys [
          {tangentMode=RCTM_User, time=0.1, value=0.15}
          {arriveTangentWeight=0.0655, interpMode=RCIM_Cubic, leaveTangentWeight=0.1021, tangentMode=RCTM_User, time=0.2, value=0.9}
          <at 0.5 0.9>
          {arriveTangentWeight=0.1021, interpMode=RCIM_Cubic, leaveTangentWeight=0.1021, tangentMode=RCTM_User, time=0.65, value=0.2}
         ]}
       }})
    )
    R2 Black:LinearColor=&MakeCustomLinearColorFromBool(
     Bool=$User.bIsFrontendPreview
     True Color=&ColorFromCurve(ColorCurve=DataInterface() {propertyValues {
        AlphaCurve {keys [<at 0.4777 1>,<at 0.4894 0>]}
        BlueCurve {keys [<at 0.12 0.004>,<at 0.71 0.006>]}
        GreenCurve {keys [<at 0.12 0.004>,<at 0.71 0.006>]}
        RedCurve {keys [<at 0.12 0.004>,<at 0.71 0.006>]}
       }})
     False Color=&ColorFromCurve(ColorCurve=DataInterface() {propertyValues {
        AlphaCurve {keys [<at 0.39 1>,<at 0.41 0>]}
        BlueCurve {keys [<at 0.1 0.006>,<at 0.7 0.01>]}
        GreenCurve {keys [<at 0.1 0.006>,<at 0.7 0.01>]}
        RedCurve {keys [<at 0.1 0.006>,<at 0.7 0.01>]}
       }})
    )
    R2 BurntSienna:LinearColor=&MakeCustomLinearColorFromBool(
     Bool=$User.bIsFrontendPreview
     True Color=&ColorFromCurve(ColorCurve=DataInterface() {propertyValues {
        AlphaCurve {keys [<at 0.4777 1>,<at 0.4894 0>]}
        BlueCurve {keys [<at 0.12 0.002>,<at 0.71 0.003>]}
        GreenCurve {keys [<at 0.12 0.0109>,<at 0.71 0.0163>]}
        RedCurve {keys [<at 0.12 0.04>,<at 0.71 0.06>]}
       }})
     False Color=&ColorFromCurve(ColorCurve=DataInterface() {propertyValues {
        AlphaCurve {keys [<at 0.39 1>,<at 0.41 0>]}
        BlueCurve {keys [<at 0.12 0.0045>,<at 0.7 0.006>]}
        GreenCurve {keys [<at 0.12 0.0245>,<at 0.7 0.0326>]}
        RedCurve {keys [<at 0.12 0.09>,<at 0.7 0.12>]}
       }})
    )
    R2 Cobalt:LinearColor=&MakeCustomLinearColorFromBool(
     Bool=$User.bIsFrontendPreview
     True Color=&ColorFromCurve(ColorCurve=DataInterface() {propertyValues {
        AlphaCurve {keys [<at 0.4777 1>,<at 0.4894 0>]}
        BlueCurve {keys [{arriveTangentWeight=0.3333, interpMode=RCIM_Cubic, leaveTangentWeight=0.3333, tangentMode=RCTM_User, time=0.12, value=0.3}, <at 0.71 0.4>]}
        GreenCurve {keys [
          {arriveTangentWeight=0.3333, interpMode=RCIM_Cubic, leaveTangentWeight=0.3333, tangentMode=RCTM_User, time=0.12, value=0.0388}
          <at 0.71 0.0517>
         ]}
        RedCurve {keys [
          {arriveTangentWeight=0.3333, interpMode=RCIM_Cubic, leaveTangentWeight=0.3333, tangentMode=RCTM_User, time=0.12, value=0.015}
          <at 0.71 0.02>
         ]}
       }})
     False Color=&ColorFromCurve(ColorCurve=DataInterface() {propertyValues {
        AlphaCurve {keys [<at 0.39 1>,<at 0.41 0>]}
        BlueCurve {keys [{arriveTangentWeight=0.3333, interpMode=RCIM_Cubic, leaveTangentWeight=0.3333, tangentMode=RCTM_User, time=0.12, value=0.3}, <at 0.7 0.5>]}
        GreenCurve {keys [
          {arriveTangentWeight=0.3333, interpMode=RCIM_Cubic, leaveTangentWeight=0.3333, tangentMode=RCTM_User, time=0.12, value=0.047}
          <at 0.7 0.0783>
         ]}
        RedCurve {keys [{arriveTangentWeight=0.3333, interpMode=RCIM_Cubic, leaveTangentWeight=0.3333, tangentMode=RCTM_User, time=0.12, value=0.024}, <at 0.7 0.04>]}
       }})
    )
    R2 Crimson:LinearColor=&MakeCustomLinearColorFromBool(
     Bool=$User.bIsFrontendPreview
     True Color=&ColorFromCurve(ColorCurve=DataInterface() {propertyValues {
        AlphaCurve {keys [<at 0.4777 1>,<at 0.4894 0>]}
        BlueCurve {keys [{arriveTangentWeight=0.3333, interpMode=RCIM_Cubic, leaveTangentWeight=0.3333, tangentMode=RCTM_User, time=0.12, value=0}, <at 0.71 0>]}
        GreenCurve {keys [
          {arriveTangentWeight=0.3333, interpMode=RCIM_Cubic, leaveTangentWeight=0.3333, tangentMode=RCTM_User, time=0.12, value=0.0008}
          <at 0.71 0.0042>
         ]}
        RedCurve {keys [{arriveTangentWeight=0.3333, interpMode=RCIM_Cubic, leaveTangentWeight=0.3333, tangentMode=RCTM_User, time=0.12, value=0.05}, <at 0.71 0.25>]}
       }})
     False Color=&ColorFromCurve(ColorCurve=DataInterface() {propertyValues {
        AlphaCurve {keys [<at 0.39 1>,<at 0.41 0>]}
        BlueCurve {keys [{arriveTangentWeight=0.3333, interpMode=RCIM_Cubic, leaveTangentWeight=0.3333, tangentMode=RCTM_User, time=0.12, value=0}, <at 0.7 0>]}
        GreenCurve {keys [
          {arriveTangentWeight=0.3333, interpMode=RCIM_Cubic, leaveTangentWeight=0.3333, tangentMode=RCTM_User, time=0.12, value=0.0033}
          <at 0.7 0.0058>
         ]}
        RedCurve {keys [{arriveTangentWeight=0.3333, interpMode=RCIM_Cubic, leaveTangentWeight=0.3333, tangentMode=RCTM_User, time=0.12, value=0.2}, <at 0.7 0.35>]}
       }})
    )
    R2 DefaultColor:LinearColor=&MakeCustomLinearColorFromBool(
     Bool=$User.bIsFrontendPreview
     True Color=&ColorFromCurve(ColorCurve=DataInterface() {propertyValues {
        AlphaCurve {keys [<at 0.4777 1>,<at 0.4894 0>]}
        BlueCurve {keys [{arriveTangentWeight=0.3333, interpMode=RCIM_Cubic, leaveTangentWeight=0.3333, tangentMode=RCTM_User, time=0.12, value=0.15}, <at 0.71 0.3>]}
        GreenCurve {keys [
          {arriveTangentWeight=0.3333, interpMode=RCIM_Cubic, leaveTangentWeight=0.3333, tangentMode=RCTM_User, time=0.12, value=0.0139}
          <at 0.71 0.0277>
         ]}
        RedCurve {keys [
          {arriveTangentWeight=0.3333, interpMode=RCIM_Cubic, leaveTangentWeight=0.3333, tangentMode=RCTM_User, time=0.12, value=0.0015}
          <at 0.71 0.003>
         ]}
       }})
     False Color=&ColorFromCurve(ColorCurve=DataInterface() {propertyValues {
        AlphaCurve {keys [<at 0.39 1>,<at 0.41 0>]}
        BlueCurve {keys [{arriveTangentWeight=0.3333, interpMode=RCIM_Cubic, leaveTangentWeight=0.3333, tangentMode=RCTM_User, time=0.1, value=0.2}, <at 0.7 0.3>]}
        GreenCurve {keys [
          {arriveTangentWeight=0.3333, interpMode=RCIM_Cubic, leaveTangentWeight=0.3333, tangentMode=RCTM_User, time=0.1, value=0.0185}
          <at 0.7 0.0277>
         ]}
        RedCurve {keys [{arriveTangentWeight=0.3333, interpMode=RCIM_Cubic, leaveTangentWeight=0.3333, tangentMode=RCTM_User, time=0.1, value=0.002}, <at 0.7 0.003>]}
       }})
    )
    R2 ForestGreen:LinearColor=&MakeCustomLinearColorFromBool(
     Bool=$User.bIsFrontendPreview
     True Color=&ColorFromCurve(ColorCurve=DataInterface() {propertyValues {
        AlphaCurve {keys [<at 0.4777 1>,<at 0.4894 0>]}
        BlueCurve {keys [
          {arriveTangentWeight=0.3333, interpMode=RCIM_Cubic, leaveTangentWeight=0.3333, tangentMode=RCTM_User, time=0.12, value=0.002}
          <at 0.71 0.0064>
         ]}
        GreenCurve {keys [{arriveTangentWeight=0.3333, interpMode=RCIM_Cubic, leaveTangentWeight=0.3333, tangentMode=RCTM_User, time=0.12, value=0.05}, <at 0.71 0.16>]}
        RedCurve {keys [
          {arriveTangentWeight=0.3333, interpMode=RCIM_Cubic, leaveTangentWeight=0.3333, tangentMode=RCTM_User, time=0.12, value=0.002}
          <at 0.71 0.0064>
         ]}
       }})
     False Color=&ColorFromCurve(ColorCurve=DataInterface() {propertyValues {
        AlphaCurve {keys [<at 0.39 1>,<at 0.41 0>]}
        BlueCurve {keys [
          {arriveTangentWeight=0.3333, interpMode=RCIM_Cubic, leaveTangentWeight=0.3333, tangentMode=RCTM_User, time=0.12, value=0.0048}
          <at 0.7 0.008>
         ]}
        GreenCurve {keys [{arriveTangentWeight=0.3333, interpMode=RCIM_Cubic, leaveTangentWeight=0.3333, tangentMode=RCTM_User, time=0.12, value=0.12}, <at 0.7 0.2>]}
        RedCurve {keys [
          {arriveTangentWeight=0.3333, interpMode=RCIM_Cubic, leaveTangentWeight=0.3333, tangentMode=RCTM_User, time=0.12, value=0.0048}
          <at 0.7 0.008>
         ]}
       }})
    )
    R2 Grey:LinearColor=&MakeCustomLinearColorFromBool(
     Bool=$User.bIsFrontendPreview
     True Color=&ColorFromCurve(ColorCurve=DataInterface() {propertyValues {
        AlphaCurve {keys [<at 0.4777 1>,<at 0.4894 0>]}
        BlueCurve {keys [{arriveTangentWeight=0.3333, interpMode=RCIM_Cubic, leaveTangentWeight=0.3333, tangentMode=RCTM_User, time=0.12, value=0.02}, <at 0.71 0.08>]}
        GreenCurve {keys [{arriveTangentWeight=0.3333, interpMode=RCIM_Cubic, leaveTangentWeight=0.3333, tangentMode=RCTM_User, time=0.12, value=0.02}, <at 0.71 0.08>]}
        RedCurve {keys [{arriveTangentWeight=0.3333, interpMode=RCIM_Cubic, leaveTangentWeight=0.3333, tangentMode=RCTM_User, time=0.12, value=0.02}, <at 0.71 0.08>]}
       }})
     False Color=&ColorFromCurve(ColorCurve=DataInterface() {propertyValues {
        AlphaCurve {keys [<at 0.39 1>,<at 0.41 0>]}
        BlueCurve {keys [{arriveTangentWeight=0.3333, interpMode=RCIM_Cubic, leaveTangentWeight=0.3333, tangentMode=RCTM_User, time=0.12, value=0.06}, <at 0.7 0.15>]}
        GreenCurve {keys [{arriveTangentWeight=0.3333, interpMode=RCIM_Cubic, leaveTangentWeight=0.3333, tangentMode=RCTM_User, time=0.12, value=0.06}, <at 0.7 0.15>]}
        RedCurve {keys [{arriveTangentWeight=0.3333, interpMode=RCIM_Cubic, leaveTangentWeight=0.3333, tangentMode=RCTM_User, time=0.12, value=0.06}, <at 0.7 0.15>]}
       }})
    )
    R2 Lime:LinearColor=&MakeCustomLinearColorFromBool(
     Bool=$User.bIsFrontendPreview
     True Color=&ColorFromCurve(ColorCurve=DataInterface() {propertyValues {
        AlphaCurve {keys [<at 0.4777 1>,<at 0.4894 0>]}
        BlueCurve {keys [
          {arriveTangentWeight=0.3333, interpMode=RCIM_Cubic, leaveTangentWeight=0.3333, tangentMode=RCTM_User, time=0.12, value=0.003}
          <at 0.71 0.0075>
         ]}
        GreenCurve {keys [{arriveTangentWeight=0.3333, interpMode=RCIM_Cubic, leaveTangentWeight=0.3333, tangentMode=RCTM_User, time=0.12, value=0.1}, <at 0.71 0.25>]}
        RedCurve {keys [
          {arriveTangentWeight=0.3333, interpMode=RCIM_Cubic, leaveTangentWeight=0.3333, tangentMode=RCTM_User, time=0.12, value=0.0596}
          <at 0.71 0.1287>
         ]}
       }})
     False Color=&ColorFromCurve(ColorCurve=DataInterface() {propertyValues {
        AlphaCurve {keys [<at 0.39 1>,<at 0.41 0>]}
        BlueCurve {keys [
          {arriveTangentWeight=0.3333, interpMode=RCIM_Cubic, leaveTangentWeight=0.3333, tangentMode=RCTM_User, time=0.12, value=0.0048}
          <at 0.7 0.0075>
         ]}
        GreenCurve {keys [{arriveTangentWeight=0.3333, interpMode=RCIM_Cubic, leaveTangentWeight=0.3333, tangentMode=RCTM_User, time=0.12, value=0.16}, <at 0.7 0.25>]}
        RedCurve {keys [
          {arriveTangentWeight=0.3333, interpMode=RCIM_Cubic, leaveTangentWeight=0.3333, tangentMode=RCTM_User, time=0.12, value=0.0953}
          <at 0.7 0.1287>
         ]}
       }})
    )
    R2 Orange:LinearColor=&MakeCustomLinearColorFromBool(
     Bool=$User.bIsFrontendPreview
     True Color=&ColorFromCurve(ColorCurve=DataInterface() {propertyValues {
        AlphaCurve {keys [<at 0.4777 1>,<at 0.4894 0>]}
        BlueCurve {keys [<at 0.12 0>,<at 0.71 0>]}
        GreenCurve {keys [<at 0.12 0.0458>,<at 0.71 0.0792>]}
        RedCurve {keys [<at 0.12 0.11>,<at 0.71 0.19>]}
       }})
     False Color=&ColorFromCurve(ColorCurve=DataInterface() {propertyValues {
        AlphaCurve {keys [<at 0.39 1>,<at 0.41 0>]}
        BlueCurve {keys [<at 0.12 0>,<at 0.7 0>]}
        GreenCurve {keys [<at 0.12 0.0833>,<at 0.7 0.125>]}
        RedCurve {keys [<at 0.12 0.2>,<at 0.7 0.3>]}
       }})
    )
    R2 Pink:LinearColor=&MakeCustomLinearColorFromBool(
     Bool=$User.bIsFrontendPreview
     True Color=&ColorFromCurve(ColorCurve=DataInterface() {propertyValues {
        AlphaCurve {keys [<at 0.4777 1>,<at 0.4894 0>]}
        BlueCurve {keys [{arriveTangentWeight=0.3333, interpMode=RCIM_Cubic, leaveTangentWeight=0.3333, tangentMode=RCTM_User, time=0.12, value=0.12}, <at 0.71 0.24>]}
        GreenCurve {keys [
          {arriveTangentWeight=0.3333, interpMode=RCIM_Cubic, leaveTangentWeight=0.3333, tangentMode=RCTM_User, time=0.12, value=0.036}
          <at 0.71 0.072>
         ]}
        RedCurve {keys [
          {arriveTangentWeight=0.3333, interpMode=RCIM_Cubic, leaveTangentWeight=0.3333, tangentMode=RCTM_User, time=0.12, value=0.099}
          <at 0.71 0.198>
         ]}
       }})
     False Color=&ColorFromCurve(ColorCurve=DataInterface() {propertyValues {
        AlphaCurve {keys [<at 0.39 1>,<at 0.41 0>]}
        BlueCurve {keys [{arriveTangentWeight=0.3333, interpMode=RCIM_Cubic, leaveTangentWeight=0.3333, tangentMode=RCTM_User, time=0.12, value=0.2}, <at 0.7 0.3>]}
        GreenCurve {keys [{arriveTangentWeight=0.3333, interpMode=RCIM_Cubic, leaveTangentWeight=0.3333, tangentMode=RCTM_User, time=0.12, value=0.06}, <at 0.7 0.09>]}
        RedCurve {keys [
          {arriveTangentWeight=0.3333, interpMode=RCIM_Cubic, leaveTangentWeight=0.3333, tangentMode=RCTM_User, time=0.12, value=0.165}
          <at 0.7 0.2475>
         ]}
       }})
    )
    R2 Purple:LinearColor=&MakeCustomLinearColorFromBool(
     Bool=$User.bIsFrontendPreview
     True Color=&ColorFromCurve(ColorCurve=DataInterface() {propertyValues {
        AlphaCurve {keys [<at 0.4777 1>,<at 0.4894 0>]}
        BlueCurve {keys [{arriveTangentWeight=0.3333, interpMode=RCIM_Cubic, leaveTangentWeight=0.3333, tangentMode=RCTM_User, time=0.12, value=0.05}, <at 0.71 0.25>]}
        GreenCurve {keys [{arriveTangentWeight=0.3333, interpMode=RCIM_Cubic, leaveTangentWeight=0.3333, tangentMode=RCTM_User, time=0.12, value=0}, <at 0.71 0>]}
        RedCurve {keys [
          {arriveTangentWeight=0.3333, interpMode=RCIM_Cubic, leaveTangentWeight=0.3333, tangentMode=RCTM_User, time=0.12, value=0.0183}
          <at 0.71 0.0917>
         ]}
       }})
     False Color=&ColorFromCurve(ColorCurve=DataInterface() {propertyValues {
        AlphaCurve {keys [<at 0.39 1>,<at 0.41 0>]}
        BlueCurve {keys [{arriveTangentWeight=0.3333, interpMode=RCIM_Cubic, leaveTangentWeight=0.3333, tangentMode=RCTM_User, time=0.12, value=0.2}, <at 0.7 0.35>]}
        GreenCurve {keys [{arriveTangentWeight=0.3333, interpMode=RCIM_Cubic, leaveTangentWeight=0.3333, tangentMode=RCTM_User, time=0.12, value=0}, <at 0.7 0>]}
        RedCurve {keys [
          {arriveTangentWeight=0.3333, interpMode=RCIM_Cubic, leaveTangentWeight=0.3333, tangentMode=RCTM_User, time=0.12, value=0.0733}
          <at 0.7 0.1283>
         ]}
       }})
    )
    R2 Saffron:LinearColor=&MakeCustomLinearColorFromBool(
     Bool=$User.bIsFrontendPreview
     True Color=&ColorFromCurve(ColorCurve=DataInterface() {propertyValues {
        AlphaCurve {keys [<at 0.4777 1>,<at 0.4894 0>]}
        BlueCurve {keys [
          {arriveTangentWeight=0.3333, interpMode=RCIM_Cubic, leaveTangentWeight=0.3333, tangentMode=RCTM_User, time=0.12, value=0.0048}
          <at 0.71 0.012>
         ]}
        GreenCurve {keys [{arriveTangentWeight=0.3333, interpMode=RCIM_Cubic, leaveTangentWeight=0.3333, tangentMode=RCTM_User, time=0.12, value=0.12}, <at 0.71 0.3>]}
        RedCurve {keys [{arriveTangentWeight=0.3333, interpMode=RCIM_Cubic, leaveTangentWeight=0.3333, tangentMode=RCTM_User, time=0.12, value=0.12}, <at 0.71 0.3>]}
       }})
     False Color=&ColorFromCurve(ColorCurve=DataInterface() {propertyValues {
        AlphaCurve {keys [<at 0.39 1>,<at 0.41 0>]}
        BlueCurve {keys [
          {arriveTangentWeight=0.3333, interpMode=RCIM_Cubic, leaveTangentWeight=0.3333, tangentMode=RCTM_User, time=0.12, value=0.0064}
          <at 0.7 0.0128>
         ]}
        GreenCurve {keys [{arriveTangentWeight=0.3333, interpMode=RCIM_Cubic, leaveTangentWeight=0.3333, tangentMode=RCTM_User, time=0.12, value=0.16}, <at 0.7 0.32>]}
        RedCurve {keys [{arriveTangentWeight=0.3333, interpMode=RCIM_Cubic, leaveTangentWeight=0.3333, tangentMode=RCTM_User, time=0.12, value=0.16}, <at 0.7 0.32>]}
       }})
    )
    R2 SkyBlue:LinearColor=&MakeCustomLinearColorFromBool(
     Bool=$User.bIsFrontendPreview
     True Color=&ColorFromCurve(ColorCurve=DataInterface() {propertyValues {
        AlphaCurve {keys [<at 0.4777 1>,<at 0.4894 0>]}
        BlueCurve {keys [{arriveTangentWeight=0.3333, interpMode=RCIM_Cubic, leaveTangentWeight=0.3333, tangentMode=RCTM_User, time=0.12, value=0.15}, <at 0.71 0.4>]}
        GreenCurve {keys [
          {arriveTangentWeight=0.3333, interpMode=RCIM_Cubic, leaveTangentWeight=0.3333, tangentMode=RCTM_User, time=0.12, value=0.0569}
          <at 0.71 0.1517>
         ]}
        RedCurve {keys [
          {arriveTangentWeight=0.3333, interpMode=RCIM_Cubic, leaveTangentWeight=0.3333, tangentMode=RCTM_User, time=0.12, value=0.003}
          <at 0.71 0.008>
         ]}
       }})
     False Color=&ColorFromCurve(ColorCurve=DataInterface() {propertyValues {
        AlphaCurve {keys [<at 0.39 1>,<at 0.41 0>]}
        BlueCurve {keys [{arriveTangentWeight=0.3333, interpMode=RCIM_Cubic, leaveTangentWeight=0.3333, tangentMode=RCTM_User, time=0.12, value=0.2}, <at 0.7 0.4>]}
        GreenCurve {keys [
          {arriveTangentWeight=0.3333, interpMode=RCIM_Cubic, leaveTangentWeight=0.3333, tangentMode=RCTM_User, time=0.12, value=0.1347}
          <at 0.7 0.2693>
         ]}
        RedCurve {keys [
          {arriveTangentWeight=0.3333, interpMode=RCIM_Cubic, leaveTangentWeight=0.3333, tangentMode=RCTM_User, time=0.12, value=0.004}
          <at 0.7 0.008>
         ]}
       }})
    )
    R2 TitaniumWhite:LinearColor=&MakeCustomLinearColorFromBool(
     Bool=$User.bIsFrontendPreview
     True Color=&ColorFromCurve(ColorCurve=DataInterface() {propertyValues {
        AlphaCurve {keys [<at 0.4777 1>,<at 0.4894 0>]}
        BlueCurve {keys [{arriveTangentWeight=0.3333, interpMode=RCIM_Cubic, leaveTangentWeight=0.3333, tangentMode=RCTM_User, time=0.12, value=0.1}, <at 0.71 0.24>]}
        GreenCurve {keys [{arriveTangentWeight=0.3333, interpMode=RCIM_Cubic, leaveTangentWeight=0.3333, tangentMode=RCTM_User, time=0.12, value=0.1}, <at 0.71 0.24>]}
        RedCurve {keys [{arriveTangentWeight=0.3333, interpMode=RCIM_Cubic, leaveTangentWeight=0.3333, tangentMode=RCTM_User, time=0.12, value=0.1}, <at 0.71 0.24>]}
       }})
     False Color=&ColorFromCurve(ColorCurve=DataInterface() {propertyValues {
        AlphaCurve {keys [<at 0.39 1>,<at 0.41 0>]}
        BlueCurve {keys [{arriveTangentWeight=0.3333, interpMode=RCIM_Cubic, leaveTangentWeight=0.3333, tangentMode=RCTM_User, time=0.12, value=0.15}, <at 0.7 0.3>]}
        GreenCurve {keys [{arriveTangentWeight=0.3333, interpMode=RCIM_Cubic, leaveTangentWeight=0.3333, tangentMode=RCTM_User, time=0.12, value=0.15}, <at 0.7 0.3>]}
        RedCurve {keys [{arriveTangentWeight=0.3333, interpMode=RCIM_Cubic, leaveTangentWeight=0.3333, tangentMode=RCTM_User, time=0.12, value=0.15}, <at 0.7 0.3>]}
       }})
    )
    R3 Black:LinearColor=&MakeCustomLinearColorFromBool(
     Bool=$User.bIsFrontendPreview
     True Color=&ColorFromCurve(ColorCurve=DataInterface() {propertyValues {
        AlphaCurve {keys [<at 0 1 RCIM_Cubic>]}
        BlueCurve {keys [<at 0.44 0.055 RCIM_Cubic>,<at 0.88 0.035 RCIM_Cubic>]}
        GreenCurve {keys [<at 0.44 0.055 RCIM_Cubic>,<at 0.88 0.035 RCIM_Cubic>]}
        RedCurve {keys [<at 0.44 0.055 RCIM_Cubic>,<at 0.88 0.035 RCIM_Cubic>]}
       }})
     False Color=&ColorFromCurve(ColorCurve=DataInterface() {propertyValues {
        AlphaCurve {keys [<at 0 1 RCIM_Cubic>]}
        BlueCurve {keys [<at 0.2 0.09 RCIM_Cubic>,<at 0.75 0.045 RCIM_Cubic>]}
        GreenCurve {keys [<at 0.2 0.09 RCIM_Cubic>,<at 0.75 0.045 RCIM_Cubic>]}
        RedCurve {keys [<at 0.2 0.09 RCIM_Cubic>,<at 0.75 0.045 RCIM_Cubic>]}
       }})
    )
    R3 BurntSienna:LinearColor=&MakeCustomLinearColorFromBool(
     Bool=$User.bIsFrontendPreview
     True Color=&ColorFromCurve(ColorCurve=DataInterface() {propertyValues {
        AlphaCurve {keys [<at 0 1 RCIM_Cubic>]}
        BlueCurve {keys [<at 0.44 0.0144 RCIM_Cubic>,<at 0.8794 0.0064 RCIM_Cubic>]}
        GreenCurve {keys [<at 0.44 0.1123 RCIM_Cubic>,<at 0.8794 0.0499 RCIM_Cubic>]}
        RedCurve {keys [<at 0.44 0.36 RCIM_Cubic>,<at 0.8794 0.16 RCIM_Cubic>]}
       }})
     False Color=&ColorFromCurve(ColorCurve=DataInterface() {propertyValues {
        AlphaCurve {keys [<at 0 1 RCIM_Cubic>]}
        BlueCurve {keys [<at 0.2 0.028 RCIM_Cubic>,<at 0.75 0.014 RCIM_Cubic>]}
        GreenCurve {keys [<at 0.2 0.2184 RCIM_Cubic>,<at 0.75 0.1092 RCIM_Cubic>]}
        RedCurve {keys [<at 0.2 0.7 RCIM_Cubic>,<at 0.75 0.35 RCIM_Cubic>]}
       }})
    )
    R3 Cobalt:LinearColor=&MakeCustomLinearColorFromBool(
     Bool=$User.bIsFrontendPreview
     True Color=&ColorFromCurve(ColorCurve=DataInterface() {propertyValues {
        AlphaCurve {keys [<at 0 1 RCIM_Cubic>]}
        BlueCurve {keys [<at 0.44 0.9 RCIM_Cubic>,<at 0.88 0.45>]}
        GreenCurve {keys [<at 0.44 0.36 RCIM_Cubic>,<at 0.88 0.18>]}
        RedCurve {keys [<at 0.44 0.18 RCIM_Cubic>,<at 0.88 0.09>]}
       }})
     False Color=&ColorFromCurve(ColorCurve=DataInterface() {propertyValues {
        AlphaCurve {keys [<at 0 1 RCIM_Cubic>]}
        BlueCurve {keys [<at 0.2 0.95 RCIM_Cubic>,<at 0.75 0.5>]}
        GreenCurve {keys [<at 0.2 0.3515 RCIM_Cubic>,<at 0.75 0.185>]}
        RedCurve {keys [<at 0.2 0.095 RCIM_Cubic>,<at 0.75 0.05>]}
       }})
    )
    R3 Crimson:LinearColor=&MakeCustomLinearColorFromBool(
     Bool=$User.bIsFrontendPreview
     True Color=&ColorFromCurve(ColorCurve=DataInterface() {propertyValues {
        AlphaCurve {keys [<at 0 1 RCIM_Cubic>]}
        BlueCurve {keys [<at 0.44 0 RCIM_Cubic>,<at 0.88 0 RCIM_Cubic>]}
        GreenCurve {keys [<at 0.44 0.0083 RCIM_Cubic>,<at 0.88 0.005 RCIM_Cubic>]}
        RedCurve {keys [<at 0.44 0.5 RCIM_Cubic>,<at 0.88 0.3 RCIM_Cubic>]}
       }})
     False Color=&ColorFromCurve(ColorCurve=DataInterface() {propertyValues {
        AlphaCurve {keys [<at 0 1 RCIM_Cubic>]}
        BlueCurve {keys [<at 0.2 0 RCIM_Cubic>,<at 0.75 0 RCIM_Cubic>]}
        GreenCurve {keys [<at 0.2 0.0142 RCIM_Cubic>,<at 0.75 0.0075 RCIM_Cubic>]}
        RedCurve {keys [<at 0.2 0.85 RCIM_Cubic>,<at 0.75 0.45 RCIM_Cubic>]}
       }})
    )
    R3 DefaultColor:LinearColor=&MakeCustomLinearColorFromBool(
     Bool=$User.bIsFrontendPreview
     True Color=&ColorFromCurve(ColorCurve=DataInterface() {propertyValues {
        AlphaCurve {keys [<at 0 1 RCIM_Cubic>]}
        BlueCurve {keys [<at 0.44 0.7913 RCIM_Cubic>,<at 0.88 0.3333 RCIM_Cubic>]}
        GreenCurve {keys [<at 0.44 0.2232 RCIM_Cubic>,<at 0.88 0.094 RCIM_Cubic>]}
        RedCurve {keys [<at 0.44 0.0319 RCIM_Cubic>,<at 0.88 0.0134 RCIM_Cubic>]}
       }})
     False Color=&ColorFromCurve(ColorCurve=DataInterface() {propertyValues {
        AlphaCurve {keys [<at 0 1 RCIM_Cubic>]}
        BlueCurve {keys [<at 0.2 0.6>,<at 0.75 0.4 RCIM_Cubic>]}
        GreenCurve {keys [<at 0.2 0.1635>,<at 0.75 0.109 RCIM_Cubic>]}
        RedCurve {keys [<at 0.2 0.018>,<at 0.75 0.012 RCIM_Cubic>]}
       }})
    )
    R3 ForestGreen:LinearColor=&MakeCustomLinearColorFromBool(
     Bool=$User.bIsFrontendPreview
     True Color=&ColorFromCurve(ColorCurve=DataInterface() {propertyValues {
        AlphaCurve {keys [<at 0 1 RCIM_Cubic>]}
        BlueCurve {keys [<at 0.44 0.016 RCIM_Cubic>,<at 0.8794 0.008 RCIM_Cubic>]}
        GreenCurve {keys [<at 0.44 0.4 RCIM_Cubic>,<at 0.8794 0.2 RCIM_Cubic>]}
        RedCurve {keys [<at 0.44 0.016 RCIM_Cubic>,<at 0.8794 0.008 RCIM_Cubic>]}
       }})
     False Color=&ColorFromCurve(ColorCurve=DataInterface() {propertyValues {
        AlphaCurve {keys [<at 0 1 RCIM_Cubic>]}
        BlueCurve {keys [<at 0.2 0.02 RCIM_Cubic>,<at 0.7 0.0104 RCIM_Cubic>]}
        GreenCurve {keys [<at 0.2 0.5 RCIM_Cubic>,<at 0.7 0.26 RCIM_Cubic>]}
        RedCurve {keys [<at 0.2 0.02 RCIM_Cubic>,<at 0.7 0.0104 RCIM_Cubic>]}
       }})
    )
    R3 Grey:LinearColor=&MakeCustomLinearColorFromBool(
     Bool=$User.bIsFrontendPreview
     True Color=&ColorFromCurve(ColorCurve=DataInterface() {propertyValues {
        AlphaCurve {keys [<at 0 1 RCIM_Cubic>]}
        BlueCurve {keys [<at 0.44 0.2 RCIM_Cubic>,<at 0.8794 0.08 RCIM_Cubic>]}
        GreenCurve {keys [<at 0.44 0.2 RCIM_Cubic>,<at 0.8794 0.08 RCIM_Cubic>]}
        RedCurve {keys [<at 0.44 0.2 RCIM_Cubic>,<at 0.8794 0.08 RCIM_Cubic>]}
       }})
     False Color=&ColorFromCurve(ColorCurve=DataInterface() {propertyValues {
        AlphaCurve {keys [<at 0 1 RCIM_Cubic>]}
        BlueCurve {keys [<at 0.2 0.3 RCIM_Cubic>,<at 0.7 0.2 RCIM_Cubic>]}
        GreenCurve {keys [<at 0.2 0.3 RCIM_Cubic>,<at 0.7 0.2 RCIM_Cubic>]}
        RedCurve {keys [<at 0.2 0.3 RCIM_Cubic>,<at 0.7 0.2 RCIM_Cubic>]}
       }})
    )
    R3 Lime:LinearColor=&MakeCustomLinearColorFromBool(
     Bool=$User.bIsFrontendPreview
     True Color=&ColorFromCurve(ColorCurve=DataInterface() {propertyValues {
        AlphaCurve {keys [<at 0 1 RCIM_Cubic>]}
        BlueCurve {keys [<at 0.44 0.015 RCIM_Cubic>,<at 0.8794 0.009 RCIM_Cubic>]}
        GreenCurve {keys [<at 0.44 0.5 RCIM_Cubic>,<at 0.8794 0.3 RCIM_Cubic>]}
        RedCurve {keys [<at 0.44 0.4192 RCIM_Cubic>,<at 0.8794 0.2515 RCIM_Cubic>]}
       }})
     False Color=&ColorFromCurve(ColorCurve=DataInterface() {propertyValues {
        AlphaCurve {keys [<at 0 1 RCIM_Cubic>]}
        BlueCurve {keys [<at 0.2 0.024 RCIM_Cubic>,<at 0.7 0.012 RCIM_Cubic>]}
        GreenCurve {keys [<at 0.2 0.8 RCIM_Cubic>,<at 0.7 0.4 RCIM_Cubic>]}
        RedCurve {keys [<at 0.2 0.6707 RCIM_Cubic>,<at 0.7 0.3353 RCIM_Cubic>]}
       }})
    )
    R3 Orange:LinearColor=&MakeCustomLinearColorFromBool(
     Bool=$User.bIsFrontendPreview
     True Color=&ColorFromCurve(ColorCurve=DataInterface() {propertyValues {
        AlphaCurve {keys [<at 0 1 RCIM_Cubic>]}
        BlueCurve {keys [<at 0.44 0 RCIM_Cubic>,<at 0.8794 0 RCIM_Cubic>]}
        GreenCurve {keys [<at 0.44 0.375 RCIM_Cubic>,<at 0.8794 0.2083 RCIM_Cubic>]}
        RedCurve {keys [<at 0.44 0.45 RCIM_Cubic>,<at 0.8794 0.25 RCIM_Cubic>]}
       }})
     False Color=&ColorFromCurve(ColorCurve=DataInterface() {propertyValues {
        AlphaCurve {keys [<at 0 1 RCIM_Cubic>]}
        BlueCurve {keys [<at 0.2 0 RCIM_Cubic>,<at 0.7 0 RCIM_Cubic>]}
        GreenCurve {keys [<at 0.2 0.6667 RCIM_Cubic>,<at 0.7 0.3333 RCIM_Cubic>]}
        RedCurve {keys [<at 0.2 0.8 RCIM_Cubic>,<at 0.7 0.4 RCIM_Cubic>]}
       }})
    )
    R3 Pink:LinearColor=&MakeCustomLinearColorFromBool(
     Bool=$User.bIsFrontendPreview
     True Color=&ColorFromCurve(ColorCurve=DataInterface() {propertyValues {
        AlphaCurve {keys [<at 0 1 RCIM_Cubic>]}
        BlueCurve {keys [<at 0.44 0.8 RCIM_Cubic>,<at 0.88 0.4>]}
        GreenCurve {keys [<at 0.44 0.4 RCIM_Cubic>,<at 0.88 0.2>]}
        RedCurve {keys [<at 0.44 0.7 RCIM_Cubic>,<at 0.88 0.35>]}
       }})
     False Color=&ColorFromCurve(ColorCurve=DataInterface() {propertyValues {
        AlphaCurve {keys [<at 0 1 RCIM_Cubic>]}
        BlueCurve {keys [<at 0.2 0.85 RCIM_Cubic>,<at 0.75 0.4>]}
        GreenCurve {keys [<at 0.2 0.425 RCIM_Cubic>,<at 0.75 0.2>]}
        RedCurve {keys [<at 0.2 0.7438 RCIM_Cubic>,<at 0.75 0.35>]}
       }})
    )
    R3 Purple:LinearColor=&MakeCustomLinearColorFromBool(
     Bool=$User.bIsFrontendPreview
     True Color=&ColorFromCurve(ColorCurve=DataInterface() {propertyValues {
        AlphaCurve {keys [<at 0 1 RCIM_Cubic>]}
        BlueCurve {keys [<at 0.44 0.7 RCIM_Cubic>,<at 0.88 0.4>]}
        GreenCurve {keys [<at 0.44 0 RCIM_Cubic>,<at 0.88 0>]}
        RedCurve {keys [<at 0.44 0.2567 RCIM_Cubic>,<at 0.88 0.1467>]}
       }})
     False Color=&ColorFromCurve(ColorCurve=DataInterface() {propertyValues {
        AlphaCurve {keys [<at 0 1 RCIM_Cubic>]}
        BlueCurve {keys [<at 0.2 0.85 RCIM_Cubic>,<at 0.75 0.45>]}
        GreenCurve {keys [<at 0.2 0 RCIM_Cubic>,<at 0.75 0>]}
        RedCurve {keys [<at 0.2 0.4108 RCIM_Cubic>,<at 0.75 0.2175>]}
       }})
    )
    R3 Saffron:LinearColor=&MakeCustomLinearColorFromBool(
     Bool=$User.bIsFrontendPreview
     True Color=&ColorFromCurve(ColorCurve=DataInterface() {propertyValues {
        AlphaCurve {keys [<at 0 1 RCIM_Cubic>]}
        BlueCurve {keys [<at 0.44 0.015 RCIM_Cubic>,<at 0.8794 0.009 RCIM_Cubic>]}
        GreenCurve {keys [<at 0.44 0.5 RCIM_Cubic>,<at 0.8794 0.3 RCIM_Cubic>]}
        RedCurve {keys [<at 0.44 0.5 RCIM_Cubic>,<at 0.8794 0.3 RCIM_Cubic>]}
       }})
     False Color=&ColorFromCurve(ColorCurve=DataInterface() {propertyValues {
        AlphaCurve {keys [<at 0 1 RCIM_Cubic>]}
        BlueCurve {keys [<at 0.2 0.021 RCIM_Cubic>,<at 0.75 0.012 RCIM_Cubic>]}
        GreenCurve {keys [<at 0.2 0.7 RCIM_Cubic>,<at 0.75 0.4 RCIM_Cubic>]}
        RedCurve {keys [<at 0.2 0.7 RCIM_Cubic>,<at 0.75 0.4 RCIM_Cubic>]}
       }})
    )
    R3 SkyBlue:LinearColor=&MakeCustomLinearColorFromBool(
     Bool=$User.bIsFrontendPreview
     True Color=&ColorFromCurve(ColorCurve=DataInterface() {propertyValues {
        AlphaCurve {keys [<at 0 1 RCIM_Cubic>]}
        BlueCurve {keys [<at 0.44 0.9 RCIM_Cubic>,<at 0.8794 0.5 RCIM_Cubic>]}
        GreenCurve {keys [<at 0.44 0.3414 RCIM_Cubic>,<at 0.8794 0.1897 RCIM_Cubic>]}
        RedCurve {keys [<at 0.44 0.018 RCIM_Cubic>,<at 0.8794 0.01 RCIM_Cubic>]}
       }})
     False Color=&ColorFromCurve(ColorCurve=DataInterface() {propertyValues {
        AlphaCurve {keys [<at 0 1 RCIM_Cubic>]}
        BlueCurve {keys [<at 0.2 0.9 RCIM_Cubic>,<at 0.75 0.6 RCIM_Cubic>]}
        GreenCurve {keys [<at 0.2 0.606 RCIM_Cubic>,<at 0.75 0.404 RCIM_Cubic>]}
        RedCurve {keys [<at 0.2 0.018 RCIM_Cubic>,<at 0.75 0.012 RCIM_Cubic>]}
       }})
    )
    R3 TitaniumWhite:LinearColor=&MakeCustomLinearColorFromBool(
     Bool=$User.bIsFrontendPreview
     True Color=&ColorFromCurve(ColorCurve=DataInterface() {propertyValues {
        AlphaCurve {keys [<at 0 1 RCIM_Cubic>]}
        BlueCurve {keys [<at 0.44 0.7 RCIM_Cubic>,<at 0.8794 0.5 RCIM_Cubic>]}
        GreenCurve {keys [<at 0.44 0.7 RCIM_Cubic>,<at 0.8794 0.5 RCIM_Cubic>]}
        RedCurve {keys [<at 0.44 0.7 RCIM_Cubic>,<at 0.8794 0.5 RCIM_Cubic>]}
       }})
     False Color=&ColorFromCurve(ColorCurve=DataInterface() {propertyValues {
        AlphaCurve {keys [<at 0 1 RCIM_Cubic>]}
        BlueCurve {keys [<at 0.2 0.85 RCIM_Cubic>,<at 0.75 0.65 RCIM_Cubic>]}
        GreenCurve {keys [<at 0.2 0.85 RCIM_Cubic>,<at 0.75 0.65 RCIM_Cubic>]}
        RedCurve {keys [<at 0.2 0.85 RCIM_Cubic>,<at 0.75 0.65 RCIM_Cubic>]}
       }})
    )
    Saffron=&MakeCustomLinearColorFromBool(
     Bool=$User.bIsFrontendPreview
     True Color=&ColorFromCurve(ColorCurve=DataInterface() {propertyValues {
        AlphaCurve {keys [<at 0 1 RCIM_Cubic>]}
        BlueCurve {keys [<at 0 0.0015>,<at 0.12 0.0051>,<at 0.9355 0.0075>]}
        GreenCurve {keys [<at 0 0.05>,<at 0.12 0.17>,<at 0.9355 0.25>]}
        RedCurve {keys [<at 0 0.05>,<at 0.12 0.17>,<at 0.9355 0.25>]}
       }})
     False Color=&ColorFromCurve(ColorCurve=DataInterface() {propertyValues {
        AlphaCurve {keys [<at 0 1 RCIM_Cubic>]}
        BlueCurve {keys [<at 0 0.0024>,<at 0.12 0.009>,<at 0.85 0.0105>]}
        GreenCurve {keys [<at 0 0.08>,<at 0.12 0.3>,<at 0.85 0.35>]}
        RedCurve {keys [<at 0 0.08>,<at 0.12 0.3>,<at 0.85 0.35>]}
       }})
    )
    SkyBlue=&MakeCustomLinearColorFromBool(
     Bool=$User.bIsFrontendPreview
     True Color=&ColorFromCurve(ColorCurve=DataInterface() {propertyValues {
        AlphaCurve {keys [<at 0 1 RCIM_Cubic>]}
        BlueCurve {keys [<at 0 0.05>,<at 0.12 0.25>,<at 0.9355 0.35>]}
        GreenCurve {keys [<at 0 0.019>,<at 0.12 0.0948>,<at 0.9355 0.1328>]}
        RedCurve {keys [<at 0 0.001>,<at 0.12 0.005>,<at 0.9355 0.007>]}
       }})
     False Color=&ColorFromCurve(ColorCurve=DataInterface() {propertyValues {
        AlphaCurve {keys [<at 0 1 RCIM_Cubic>]}
        BlueCurve {keys [<at 0 0.08>,<at 0.12 0.25>,<at 0.85 0.35>]}
        GreenCurve {keys [<at 0 0.0539>,<at 0.12 0.1683>,<at 0.85 0.2357>]}
        RedCurve {keys [<at 0 0.0016>,<at 0.12 0.005>,<at 0.85 0.007>]}
       }})
    )
    TitaniumWhite=&MakeCustomLinearColorFromBool(
     Bool=$User.bIsFrontendPreview
     True Color=&ColorFromCurve(ColorCurve=DataInterface() {propertyValues {
        AlphaCurve {keys [<at 0 1 RCIM_Cubic>]}
        BlueCurve {keys [<at 0 0.05>,<at 0.12 0.15>,<at 0.9355 0.25>]}
        GreenCurve {keys [<at 0 0.05>,<at 0.12 0.15>,<at 0.9355 0.25>]}
        RedCurve {keys [<at 0 0.05>,<at 0.12 0.15>,<at 0.9355 0.25>]}
       }})
     False Color=&ColorFromCurve(ColorCurve=DataInterface() {propertyValues {
        AlphaCurve {keys [<at 0 1 RCIM_Cubic>]}
        BlueCurve {keys [<at 0 0.08>,<at 0.12 0.25>,<at 0.85 0.35>]}
        GreenCurve {keys [<at 0 0.08>,<at 0.12 0.25>,<at 0.85 0.35>]}
        RedCurve {keys [<at 0 0.08>,<at 0.12 0.25>,<at 0.85 0.35>]}
       }})
    )
    Use R1 Painted Overrides=true
    Use R2 Painted Overrides=true
    Use R3 Painted Overrides=true
   ) {module=VC_PaintedVariantsOverride001}
   &ParticleState()
   &ScaleSpriteSize(
    Scale Sprite Size Mode=::NewEnumerator0
    Uniform Scale Factor=&MergeHelper_FloatSwitch(
     0=&FloatFromCurve(FloatCurve=DataInterface() {propertyValues {Curve {keys [{tangentMode=RCTM_User, time=0, value=0.5}, <at 0.25 2>, {tangentMode=RCTM_User, time=1, value=1}]}}})
     1=&FloatFromCurve(FloatCurve=DataInterface() {propertyValues {Curve {keys [
          {tangentMode=RCTM_User, time=0, value=0.3}
          {arriveTangentWeight=0.1985, interpMode=RCIM_Cubic, leaveTangentWeight=0.2646, tangentMode=RCTM_User, time=0.3, value=1.5}
          {arriveTangentWeight=0.1639, interpMode=RCIM_Cubic, leaveTangentWeight=0.1229, tangentMode=RCTM_User, time=0.6, value=1.5}
          {arriveTangent=-0.9048, tangentMode=RCTM_User, time=1, value=1}
         ]}}})
     2=&FloatFromCurve(FloatCurve=DataInterface() {propertyValues {Curve {keys [{tangentMode=RCTM_User, time=0, value=0.6}, <at 0.22 2>, {tangentMode=RCTM_User, time=1, value=1}]}}})
     3=?
     Index=&SpawnGroup()
     Num=::NewEnumerator1
    )
   )
   &InheritSourceMovement(ApplyForce=false, ApplyPosition=true, ApplyPositionFalloff=false)
   &ScaleVelocity(
    Coordinate Space:ENiagaraCoordinateSpace=::Simulation
    Velocity Scale=&VectorFromFloat(Value=&MergeHelper_FloatSwitch(
      0=&FloatFromCurve(FloatCurve=DataInterface() {propertyValues {Curve {keys [
           {interpMode=RCIM_Cubic, leaveTangentWeight=0.2323, tangentMode=RCTM_User, time=0.19, value=0.19}
           {arriveTangent=1.7206, interpMode=RCIM_Cubic, leaveTangent=1.7206, tangentMode=RCTM_User, time=0.9, value=1}
          ]}}})
      1=&FloatFromCurve(FloatCurve=DataInterface() {propertyValues {Curve {keys [
           {interpMode=RCIM_Cubic, leaveTangentWeight=0.2667, tangentMode=RCTM_User, time=0.2, value=0.3}
           {arriveTangent=1.2132, arriveTangentWeight=0.2667, interpMode=RCIM_Cubic, leaveTangent=1.2132, tangentMode=RCTM_User, time=0.9, value=1}
          ]}}})
      2=&FloatFromCurve(FloatCurve=DataInterface() {propertyValues {Curve {keys [
           {interpMode=RCIM_Cubic, leaveTangentWeight=0.25, tangentMode=RCTM_User, time=0.16, value=0.26}
           {arriveTangent=1.3636, arriveTangentWeight=0.2667, interpMode=RCIM_Cubic, leaveTangent=1.3636, tangentMode=RCTM_User, time=0.9, value=1}
          ]}}})
      3=?
      Index=$User.GameModeID
      Num=::NewEnumerator1
     ))
   )
   &SolveForcesAndVelocity(Clamp Velocity=false, Limit Acceleration=false)
   &SubUVAnimation(
    End Frame Range Override=7
    "Sprite Renderer "=DataInterface() {propertyValues {SpriteRenderer=&self:NE_VC_Boost_Sprite_Parent.NiagaraSpriteRendererProperties_3}}
    UseEndFrame=false
    UseStartFrame=false
   )
  ]
  rendererClasses [&NiagaraSpriteRendererProperties]
  renderers [
   {rendererClass=&NiagaraSpriteRendererProperties}
   {rendererClass=&NiagaraSpriteRendererProperties, rendererIndex=1}
   {rendererClass=&NiagaraSpriteRendererProperties, rendererIndex=2}
   {rendererClass=&NiagaraSpriteRendererProperties, rendererIndex=3}
   {rendererClass=&NiagaraSpriteRendererProperties, rendererIndex=4}
  ]
 }
}
summary {
 systemName=NS_Abstractfire
 emitters [
  {emitterName=NE_VC_Boost_Ribbon_Parent, rendererClasses [&NiagaraRibbonRendererProperties]}
  {emitterName=NE_VC_Boost_Sprite_Parent, rendererClasses [&NiagaraSpriteRendererProperties]}
 ]
 userVariables [
  User.bIsFrontendPreview {type=NiagaraBool}
  User.bIsPainted {type=NiagaraBool}
  User.ColorVariantID {type=NiagaraInt32}
  User.DTCarOffset {type=Vector3f, description="", defaultValue {struct=Vector3f, value=<0 0 0>}}
  User.GameModeID {type=NiagaraInt32}
  User.PaintedColor {type=LinearColor, description="", defaultValue {struct=LinearColor, value=<rgba 1 1 1 1>}}
  User.PaintedScalar {type=NiagaraFloat, defaultValue=1, description=""}
  User.PortCount {type=NiagaraInt32, defaultValue=2}
 ]
}
